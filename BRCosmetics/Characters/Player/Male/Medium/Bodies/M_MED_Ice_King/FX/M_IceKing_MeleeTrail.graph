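# Translucent, unlit, two-sided beam-trail material for the Ice King character's melee trail effect.
# Emissive combines two panning dust textures into ice-tinted sparkles over a gradient texture, with HDR intensity driven by sparkle strength.
# Opacity blends panning light-ray streaks through a noise-perturbed lengthwise falloff, shaped by UV gradients and an inverted sphere mask.
# Near-camera fade, double-sided fresnel and particle alpha scale the final opacity.
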Material /BRCosmetics/Characters/Player/Male/Medium/Bodies/M_MED_Ice_King/FX/M_IceKing_MeleeTrail
graph {
 properties {
  bUsedWithBeamTrails=true
  blendMode=BLEND_Translucent
  pixelDepthOffsetMode=PDOM_Legacy
  shadingModel=MSM_Unlit
  twoSided=true
  usedShadingModels=Unlit
 }
 nodes [
  n1 = ComponentMask(None=VectorParameter() {parameterName=SparkleScale02, defaultValue=<rgba 1 0.125 5 5>}.RGB) {g=true, r=true}
  n2 = TextureCoordinate()
  n3 = TextureSample(UVs=Panner(Coordinate=Multiply(A=@n1, B=@n2)) {speedX=-0.025, speedY=0.025}) {texture=/Game/Global/Textures/Default/PointSplatter/T_Tiling_Dust}
  n4 = ComponentMask(None=VectorParameter() {parameterName=SparkleScale, defaultValue=<rgba 3 0.5 5 5>}.RGB) {g=true, r=true}
  n5 = TextureSample(UVs=Panner(Coordinate=Multiply(A=@n2, B=@n4)) {speedX=-0.025, speedY=-0.0625}) {texture=&T_Tiling_Dust}
  n6 = Multiply(
   A=Power(Base=Multiply(A=@n3.R, B=@n5.R), Exp=ScalarParameter() {defaultValue=1.75, parameterName=SparklesPower}) {constExponent=0.2}
   B=ScalarParameter() {defaultValue=375, parameterName=SparklesInstence}
  ) {constB=500}
  n7 = ComponentMask(None=TextureCoordinate()) {g=true}
  n8 = LinearInterpolate(A=OneMinus(None=@n7), B=@n7, Alpha=ScalarParameter() {parameterName=GradFlip})
  n9 = Add(
   A=Multiply(A=@n6, B=VectorParameter() {parameterName=Ice Color, defaultValue=<rgba 0.8 0.8213 1 1>}.RGB)
   B=Multiply(
    A=TextureSampleParameter2D(UVs=@n8) {parameterName=GradientTexture, texture=/Game/Global/Textures/Default/Masks_and_Gradients/HeistColorGradients/CrystalGradient}.RGB
    B=ScalarParameter() {defaultValue=1, parameterName=GradIntensity}
   )
  )
  n10 = ScalarParameter() {defaultValue=1, parameterName=HDR_Intensity_Min}
  n11 = ScalarParameter() {defaultValue=2, parameterName=HDR_Intensity_Max}
  MaterialFunctionCall(Emissive=MaterialFunctionCall(LDR Input=Clamp(None=@n9), HDR Tint=@n9, HDR Intensity Multiplier=LinearInterpolate(A=@n10, B=@n11, Alpha=@n6) {constB=10}) {materialFunction=&DeriveHDRfromLDR}.RemappedRange) {materialFunction=&MF_FX_EmissiveAdjustments}
  n12 = Panner(Coordinate=MaterialFunctionCall(XY=TextureCoordinate() {uTiling=3.464}) {materialFunction=&Swizzle}.YX) {speedX=0.976}
  n13 = TextureCoordinate() {uTiling=4.4362}
  n14 = TextureSample(UVs=Panner(Coordinate=MaterialFunctionCall(XY=@n13) {materialFunction=&Swizzle}.YX) {speedX=-0.7616}) {samplerSource=SSM_Wrap_WorldGroupSettings, texture=/Game/Effects/Fort_Effects/Textures/Outlander/T_InTheZone_LightRays_01}
  n15 = Add(
   A=Power(Base=TextureSample(UVs=@n12) {samplerSource=SSM_Wrap_WorldGroupSettings, texture=&T_InTheZone_LightRays_01}.R) {constExponent=1.5}
   B=Multiply(A=@n14.G, B=@n14.G)
  )
  n16 = TextureCoordinate() {uTiling=0.2}
  n17 = ConstantBiasScale(None=TextureSample(UVs=Panner(Coordinate=@n16) {bFractionalPart=true, speedX=-0.65}) {samplerSource=SSM_Wrap_WorldGroupSettings, texture=&T_NoisePack_01}.R) {bias=-0.5, scale=0.2}
  n18 = OneMinus(None=ComponentMask(None=TextureCoordinate()) {r=true})
  n19 = LinearInterpolate(
   A=LinearInterpolate(
    A=ScalarParameter() {parameterName=Fine Streaks Opacity A}
    B=ScalarParameter() {defaultValue=0.2, parameterName=Fine Streaks Opacity B}
    Alpha=Multiply(A=@n15) {constB=0.5}
   ) {constA=1}
   B=ScalarParameter() {defaultValue=0.8, parameterName=Hard Streaks Opacity}
   Alpha=Power(Base=Add(A=@n17, B=@n18)) {constExponent=4}
  ) {constAlpha=0}
  n20 = MaterialFunctionCall(Float2=TextureCoordinate()) {materialFunction=&BreakOutFloat2Components}
  n21 = Saturate(None=Multiply(A=@n20.R) {constB=3})
  n22 = OneMinus(None=@n20.R)
  n23 = Multiply(A=Add(B=Clamp(None=@n19)), B=Multiply(A=Multiply(A=@n21, B=@n22), B=OneMinus(None=@n20.G)))
  n24 = OneMinus(None=SphereMask(A=TextureCoordinate() {uTiling=0.5}, B=Constant2Vector() {g=0.95, r=0.85}, Radius=Constant() {r=0.7}, Hardness=Constant() {r=0.3}))
  n25 = Multiply(A=@n24, B=@n24)
  n26 = Clamp(None=Multiply(A=@n24, B=Multiply(A=@n25, B=@n25)))
  n27 = Saturate(None=Multiply(A=Multiply(A=@n23, B=@n26)) {constB=3})
  n28 = MaterialFunctionCall(Distance=ScalarParameter() {defaultValue=64, parameterName=NearFadeDistance}) {materialFunction=&Near_Camera_Fade}
  n29 = ScalarParameter() {defaultValue=1.5, parameterName=Fresnel Brightness}
  n30 = ScalarParameter() {defaultValue=3, parameterName=Fresnel Exponent}
  Multiply(
   A=ParticleColor().A
   B=Multiply(
    A=Multiply(A=@n27, B=@n28.Result)
    B=MaterialFunctionCall(Brightness=@n29, Exponent=@n30) {materialFunction=&MF_Fresnel_FX}.Fresnel Double Sided
   )
  )
 ]
}
