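# Niagara effect for the Ice King character's backpack, pairing animation trails with cold mist on a user-supplied skeletal mesh.
# FN_AnimTrail spawns two short-lived ribbon trails from bone or socket index pairs, drawn by two ribbon renderers, fading and widening over life.
# Mist spawns sprites by rate and distance moved, partitioned across four bone or socket indices.
# The mist drifts down, grows, plays a SubUV flipbook and takes its tint from User.MistColor.
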NiagaraSystem /BRCosmetics/Characters/Player/Male/Medium/Bodies/M_MED_Ice_King/FX/NS_IceKingBackpack
dependencies {
 usedDataInterfaces [
  &NiagaraDataInterfaceCurve
  &NiagaraDataInterfaceVector2DCurve
  &NiagaraDataInterfaceSpriteRendererInfo
 ]
 usedDynamicInputs []
 usedModules [&SystemState]
}
emitters {
 FN_AnimTrail {
  emitterSpawnScript [&FN_RibbonIDOffsetUtility()]
  emitterUpdateScript [
   &EmitterState()
   &FN_SpawnAnimTrail(Number of Anim Trails=2)
  ]
  particleSpawnScript [
   &FN_InitializeAnimTrailBoneIndecies()
   &FN_InitializeAnimTrailBoneIndecies(AnimTrail Selection=1, First Bone or Socket Index=2, Second Bone or Socket Index=3) {module=FN_InitializeAnimTrailBoneIndecies001}
   &FN_InitializeAnimTrail(Color=<rgba 1 1 1 0.6>, Lifetime=0.4)
   &AddVelocity(Velocity=<0 400 0>) {enabled=false}
  ]
  particleUpdateScript [
   &SolveForcesAndVelocity(Clamp Velocity=false, Limit Acceleration=false)
   &FN_UpdateAnimTrail(Locked Particle RibbonWidth=1, SkeletalMesh=$User.SkeletalMesh) {module=FN_UpdateAnimTrail001}
   &ScaleColor(
    Scale Alpha=&FloatFromCurve(FloatCurve=DataInterface() {propertyValues {Curve {keys [
         {arriveTangent=9.0725, interpMode=RCIM_Cubic, leaveTangent=9.0725, tangentMode=RCTM_User, time=0, value=0}
         {arriveTangent=-2.5249, interpMode=RCIM_Cubic, leaveTangent=-2.5249, tangentMode=RCTM_User, time=0.1408, value=0.9847}
         {arriveTangent=-0.0405, interpMode=RCIM_Cubic, leaveTangent=-0.0405, tangentMode=RCTM_User, time=1, value=0}
        ]}}})
    ScaleA=true
    ScaleRGB=true
   )
   &ScaleRibbonWidth(Ribbon Width Scale=&FloatFromCurve(FloatCurve=DataInterface() {propertyValues {Curve {keys [{arriveTangent=2.5368, interpMode=RCIM_Cubic, leaveTangent=2.5368, tangentMode=RCTM_User, time=0, value=0.3}, <at 1 2 RCIM_Cubic>]}}}))
  ]
  rendererClasses [&NiagaraRibbonRendererProperties]
  renderers [
   {rendererClass=&NiagaraRibbonRendererProperties}
   {rendererClass=&NiagaraRibbonRendererProperties, rendererIndex=1}
  ]
 }
 Mist {
  emitterUpdateScript [
   &EmitterState()
   &SpawnPerUnit(
    Module.Use Max Movement Threshold=true
    Module.Use Movement Tolerance=true
    Movement Tolerance=0.1
    Spawn Group=1
    Spawn Spacing=30
    Use Spawn Probability=false
   )
   &SpawnRate(SpawnRate=48, Use Spawn Probability=false)
  ]
  particleSpawnScript [
   &InitializeParticle(
    Color=&MakeLinearColorFromVectorAndFloat(
     "Vector (RGB)"=&VectorFromFloat(Value=&RandomRangeFloat(Minimum=0.5, Maximum=2))
     "Float (Alpha)"=0.025
    )
    Lifetime Max=1.25
    Lifetime Mode=::NewEnumerator1
    Sprite Rotation Mode=::NewEnumerator2
    Sprite Size Mode=::NewEnumerator1
    Uniform Sprite Size Max=20
    UsePositionOffset=false
    Write Material Random=true
   )
   &DynamicMaterialParameters(
    Index 0 Param 1=8
    Index 0 Param 2=&RandomRangeFloat(Minimum=0, Maximum=2)
    Index 0 Param 4=0.5
    Module.Param0WriteEnabled=true
    Module.Param1WriteEnabled=true
    Module.Param2WriteEnabled=true
    Module.Param3WriteEnabled=true
   )
   &PartitionParticles(Number Of Partitions=4)
   &CameraOffset(Camera Offset Amount=25)
  ]
  particleUpdateScript [
   &ParticleState()
   &ScaleSpriteSize(
    Non-Uniform Curve Sprite Scale=DataInterface() {propertyValues {
      XCurve {keys [{tangentMode=RCTM_User, time=0.1, value=1}, {tangentMode=RCTM_User, time=1, value=3}]}
      YCurve {keys [{tangentMode=RCTM_User, time=0.1, value=1.5}, {tangentMode=RCTM_User, time=1, value=3}]}
     }}
    Scale Sprite Size Mode=::NewEnumerator3
   )
   &ScaleColor(
    Scale Alpha=&FloatFromCurve(FloatCurve=DataInterface() {propertyValues {Curve {keys [
         {interpMode=RCIM_Cubic, tangentMode=RCTM_User, time=0, value=0}
         {interpMode=RCIM_Cubic, tangentMode=RCTM_User, time=0.5, value=1}
         {tangentMode=RCTM_User, time=1, value=0}
        ]}}})
    Scale RGB=&MakeVectorFromLinearColorRGB(Color=$User.MistColor)
    ScaleA=true
    ScaleRGB=true
   )
   &SubUVAnimation(
    Direct Frame Index:NiagaraFloat=&FloatFromCurve(FloatCurve=DataInterface() {propertyValues {Curve {keys [{arriveTangent=34.442, interpMode=RCIM_Cubic, leaveTangent=34.442, tangentMode=RCTM_User, time=0, value=0}, <at 0.9755 25.3627 RCIM_Cubic>]}}})
    "Sprite Renderer "=DataInterface() {propertyValues {SpriteRenderer=&self:Mist_0.NiagaraSpriteRendererProperties_0}}
    SubUV Animation Mode:ENiagara_SubUVLookupModeV2=::NewEnumerator5
   )
   &SkeletalMeshLocation(
    Bone / Socket Index=$Particles.PartitionParticles.Partition
    Bone Sampling Mode=::NewEnumerator4
    MaskSpawnGroup=false
    Position Sampling=::NewEnumerator1
    Skeletal Mesh=$User.SkeletalMesh
    Source Movement Scale:Vector3f=&MaskVectorBySpawnGroup(Masked Vector=<1 1 1>, Passthrough Vector=<0 0 0>)
   )
   &AccelerationForce(Acceleration=&RandomRangeVector(Minimum=<0 0 -15>, Maximum=<0 0 -75>))
   &Drag(Drag=5, Use Linear Drag=true, Use Rotational Drag=true)
   &SolveForcesAndVelocity(Clamp Velocity=false, Limit Acceleration=false)
  ]
  rendererClasses [&NiagaraSpriteRendererProperties]
  renderers [{rendererClass=&NiagaraSpriteRendererProperties}]
 }
}
summary {
 systemName=NS_IceKingBackpack
 emitters [
  {emitterName=FN_AnimTrail, rendererClasses [&NiagaraRibbonRendererProperties]}
  {emitterName=Mist, rendererClasses [&NiagaraSpriteRendererProperties]}
 ]
 userVariables [
  User.Color_Stripes_Bot {type=LinearColor, description="", defaultValue {struct=LinearColor, value=<rgba 0.0493 0.553 1 1>}}
  User.Color_Stripes_Mid {type=LinearColor, description="", defaultValue {struct=LinearColor, value=<rgba 0.0493 0.553 1 1>}}
  User.Color_Stripes_Top {type=LinearColor, description="", defaultValue {struct=LinearColor, value=<rgba 0.0493 0.553 1 1>}}
  User.MistColor {type=LinearColor, description="", defaultValue {struct=LinearColor, value=<rgba 0.0493 0.553 1 1>}}
  User.RibbonIDOffset {type=NiagaraInt32}
  User.SkeletalMesh {
   type=NiagaraDataInterfaceSkeletalMesh
   defaultValue {
    struct=NiagaraExt_VariableValue_DataInterface
    value {dataInterface=&self:NiagaraDataInterfaceSkeletalMesh_0}
   }
  }
 ]
}
