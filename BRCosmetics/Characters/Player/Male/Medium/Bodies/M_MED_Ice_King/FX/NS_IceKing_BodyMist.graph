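# Niagara body mist for the Ice King character skin, tinted by User.MistColor (icy blue default) and attached to User.SkeletalMesh.
# Smoke continuously spawns sprites in two groups, partitioned across 13 bone or socket indices with positional offsets, using spawn-group masks to vary downward acceleration and drag.
# VelocitySmoke spawns by distance moved at partition-offset bones, adding sinking, speed-stretched puffs with decaying inherited movement.
# Both emitters use curve-driven flipbook frames, growth and alpha fades.
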NiagaraSystem /BRCosmetics/Characters/Player/Male/Medium/Bodies/M_MED_Ice_King/FX/NS_IceKing_BodyMist
dependencies {
 usedDataInterfaces [
  &NiagaraDataInterfaceCurve
  &NiagaraDataInterfaceVector2DCurve
  &NiagaraDataInterfaceSpriteRendererInfo
  &NiagaraDataInterfaceArrayFloat3
 ]
 usedDynamicInputs []
 usedModules [&SystemState]
}
emitters {
 Smoke {
  emitterUpdateScript [
   &EmitterState()
   &SpawnRate(SpawnRate=13, Use Spawn Probability=false)
   &SpawnRate(SpawnGroup=1, SpawnRate=13, Use Spawn Probability=false) {module=SpawnRate001}
  ]
  particleSpawnScript [
   &InitializeParticle(
    Color=&MakeLinearColorFromVectorAndFloat(
     "Vector (RGB)"=&VectorFromFloat(Value=&RandomRangeFloat(Minimum=0.5, Maximum=1))
     "Float (Alpha)"=&Lerp_Float(
      A=&RandomRangeFloat(Minimum=0.1, Maximum=0.2)
      B=&RandomRangeFloat(Minimum=0.05, Maximum=0.15)
      Alpha=&FloatFromCurve(
       FloatCurve=DataInterface() {propertyValues {Curve {keys [{arriveTangent=1.7029, interpMode=RCIM_Cubic, leaveTangent=1.7029, tangentMode=RCTM_User, time=0, value=0}, <at 1 1 RCIM_Cubic>]}}}
       CurveIndex=$Emitter.NormalizedLoopAge
      )
     )
    )
    Lifetime Max=2.25
    Lifetime Min=1.75
    Lifetime Mode=::NewEnumerator1
    Sprite Rotation Mode=::NewEnumerator2
    Sprite Size Mode=::NewEnumerator1
    Sprite UV Mode=::NewEnumerator1
    Uniform Sprite Size Max=20
    UsePositionOffset=false
    Write Material Random=true
   )
   &PartitionParticles(Number Of Partitions=13)
   &CameraOffset(Camera Offset Amount=15)
  ]
  particleUpdateScript [
   &ParticleState()
   &ScaleSpriteSize(
    Non-Uniform Curve Sprite Scale=DataInterface() {propertyValues {
      XCurve {keys [{tangentMode=RCTM_User, time=0, value=1}, {tangentMode=RCTM_User, time=1, value=2}]}
      YCurve {keys [{tangentMode=RCTM_User, time=0, value=1}, <at 0.1 1.5>, {tangentMode=RCTM_User, time=1, value=2}]}
     }}
    Scale Sprite Size Mode=::NewEnumerator3
   )
   &ScaleColor(
    Scale Alpha=&FloatFromCurve(FloatCurve=DataInterface() {propertyValues {Curve {keys [
         {arriveTangent=3.7506, interpMode=RCIM_Cubic, leaveTangent=3.7506, tangentMode=RCTM_User, time=0, value=0}
         {interpMode=RCIM_Cubic, leaveTangent=-2.614, tangentMode=RCTM_Break, time=0.5, value=1}
         {tangentMode=RCTM_User, time=1, value=0}
        ]}}})
    Scale RGB=&MakeVectorFromLinearColorRGB(Color=$User.MistColor)
    ScaleA=true
    ScaleRGB=true
   )
   &SubUVAnimation(
    Direct Frame Index:NiagaraFloat=&FloatFromCurve(FloatCurve=DataInterface() {propertyValues {Curve {keys [<at 0 0 RCIM_Cubic>, {arriveTangent=85.3311, interpMode=RCIM_Cubic, leaveTangent=85.3311, tangentMode=RCTM_User, time=1, value=45.5498}]}}})
    "Sprite Renderer "=DataInterface() {propertyValues {SpriteRenderer=&self:Smoke_1.NiagaraSpriteRendererProperties_0}}
    SubUV Animation Mode:ENiagara_SubUVLookupModeV2=::NewEnumerator5
   )
   &SkeletalMeshLocation(
    Bone / Socket Index=$Particles.PartitionParticles.Partition
    Bone Sampling Mode=::NewEnumerator4
    "Inherit Velocity (Scale)"=0
    MaskSpawnGroup=false
    Position Sampling=::NewEnumerator1
    Sampled Position Offset=&SelectVectorFromArray(
     Direct Array Index=$Particles.PartitionParticles.Partition
     Vector Selection Array=DataInterface() {propertyValues {FloatData [<0 20 0>, <0 20 0>, <0 -20 0>, <0 15 0>, <0 15 0>, <0 15 0>, <10 0 0>, <10 0 0>, <10 0 0>, <10 0 0>, <10 0 0>, <10 0 0>, <10 0 0>]}}
    )
    Skeletal Mesh=$User.SkeletalMesh
    Velocity Sampling=::NewEnumerator0
    VelocitySpeedLimit=false
   )
   &DynamicMaterialParameters(
    Index 0 Param 1=1
    Index 0 Param 2=&RandomRangeFloat(Minimum=0, Maximum=2)
    Index 0 Param 4=0.5
    Module.Param0WriteEnabled=true
    Module.Param1WriteEnabled=true
    Module.Param2WriteEnabled=true
    Module.Param3WriteEnabled=true
   )
   &AccelerationForce(Acceleration=&MaskVectorBySpawnGroup(Masked Vector=&RandomRangeVector(Minimum=<0 0 -75>, Maximum=<0 0 -150>), Passthrough Vector=<0 0 0>))
   &Drag(
    Drag=&MaskFloatBySpawnGroup(Masked Float=8, Passthrough Float=2.5)
    Use Linear Drag=true
    Use Rotational Drag=true
   )
   &SolveForcesAndVelocity(Clamp Velocity=false, Limit Acceleration=false)
  ]
  rendererClasses [&NiagaraSpriteRendererProperties]
  renderers [{rendererClass=&NiagaraSpriteRendererProperties}]
 }
 VelocitySmoke {
  emitterUpdateScript [
   &EmitterState()
   &SpawnPerUnit(
    Module.Use Max Movement Threshold=true
    Module.Use Movement Tolerance=true
    Movement Tolerance=0.1
    Spawn Spacing=50
    Use Spawn Probability=false
   )
  ]
  particleSpawnScript [
   &InitializeParticle(
    Color=&MakeLinearColorFromVectorAndFloat(
     "Vector (RGB)"=&VectorFromFloat(Value=&RandomRangeFloat(Minimum=0.5, Maximum=1))
     "Float (Alpha)"=&RandomRangeFloat(Minimum=0.05, Maximum=0.1)
    )
    Lifetime Max=1.2
    Lifetime Min=0.8
    Lifetime Mode=::NewEnumerator1
    Sprite Size Max=<90 90>
    Sprite Size Min=<60 60>
    Sprite Size Mode=::NewEnumerator2
    UsePositionOffset=false
    Write Material Random=true
   )
   &CameraOffset(Camera Offset Amount=&RandomRangeFloat(Minimum=15, Maximum=30))
   &AddVelocity(Velocity=&RandomRangeVector(Minimum=<-15 -15 0>, Maximum=<15 15 0>))
   &PartitionParticles()
  ]
  particleUpdateScript [
   &ParticleState()
   &SkeletalMeshLocation(
    Bone / Socket Index=&Add_Integer(A=$Particles.PartitionParticles.Partition, B=11)
    Bone Sampling Mode=::NewEnumerator4
    MaskSpawnGroup=false
    Position Sampling=::NewEnumerator1
    Skeletal Mesh=$User.SkeletalMesh
    Source Movement Scale:Vector3f=&Multiply_VectorByFloat(
     Vector=<0.25 0.25 0.25>
     Float=&FloatFromCurve(FloatCurve=DataInterface() {propertyValues {Curve {keys [{arriveTangent=-1.4888, interpMode=RCIM_Cubic, leaveTangent=-1.4888, tangentMode=RCTM_User, time=0, value=1}, <at 1 0 RCIM_Cubic>]}}})
    )
   )
   &ScaleSpriteSize(
    Non-Uniform Curve Sprite Scale=DataInterface() {propertyValues {
      XCurve {keys [
        {arriveTangent=3.2621, interpMode=RCIM_Cubic, leaveTangent=3.2621, tangentMode=RCTM_User, time=0, value=0.05}
        {arriveTangent=1.7776, interpMode=RCIM_Cubic, leaveTangent=1.7776, time=0.2344, value=0.7419}
        {arriveTangent=0.5984, interpMode=RCIM_Cubic, leaveTangent=0.5984, time=0.5916, value=1.1017}
        <at 1 1.2 RCIM_Cubic>
       ]}
      YCurve {keys [
        {arriveTangent=4.3792, interpMode=RCIM_Cubic, leaveTangent=4.3792, tangentMode=RCTM_User, time=0, value=0.1}
        {arriveTangent=2.17, interpMode=RCIM_Cubic, leaveTangent=2.17, time=0.2344, value=1.0157}
        {arriveTangent=0.6325, interpMode=RCIM_Cubic, leaveTangent=0.6325, time=0.5916, value=1.3838}
        <at 1 1.5 RCIM_Cubic>
       ]}
     }}
    Scale Sprite Size Mode=::NewEnumerator3
   )
   &ScaleColor(
    Scale Alpha=&FloatFromCurve(FloatCurve=DataInterface() {propertyValues {Curve {keys [
         {interpMode=RCIM_Cubic, tangentMode=RCTM_User, time=0, value=0}
         {interpMode=RCIM_Cubic, tangentMode=RCTM_User, time=0.5, value=1}
         {tangentMode=RCTM_User, time=1, value=0}
        ]}}})
    Scale RGB=&MakeVectorFromLinearColorRGB(Color=$User.MistColor)
    ScaleA=true
    ScaleRGB=true
   )
   &DynamicMaterialParameters(
    Index 0 Param 1=2
    Index 0 Param 2=&RandomRangeFloat(Minimum=1, Maximum=2)
    Index 0 Param 4=0.5
    Module.Param0WriteEnabled=true
    Module.Param1WriteEnabled=true
    Module.Param2WriteEnabled=true
    Module.Param3WriteEnabled=true
   )
   &AccelerationForce(Acceleration=&RandomRangeVector(Minimum=<0 0 -30>, Maximum=<0 0 -80>))
   &Drag(Drag=1.5, Use Linear Drag=true, Use Rotational Drag=true)
   &SolveForcesAndVelocity(Clamp Velocity=false, Limit Acceleration=false)
   &ScaleSpriteSizeBySpeed(Max Scale Factor=<1 2>)
   &SubUVAnimation(
    Direct Frame Index:NiagaraFloat=&FloatFromCurve(FloatCurve=DataInterface() {propertyValues {Curve {keys [{arriveTangent=30.5428, interpMode=RCIM_Cubic, leaveTangent=30.5428, tangentMode=RCTM_User, time=0, value=9}, <at 1 39 RCIM_Cubic>]}}})
    "Sprite Renderer "=DataInterface() {propertyValues {SpriteRenderer=&self:VelocitySmoke_0.NiagaraSpriteRendererProperties_1}}
    SubUV Animation Mode:ENiagara_SubUVLookupModeV2=::NewEnumerator5
   )
  ]
  rendererClasses [&NiagaraSpriteRendererProperties]
  renderers [{rendererClass=&NiagaraSpriteRendererProperties}]
 }
}
summary {
 systemName=NS_IceKing_BodyMist
 emitters [
  {emitterName=Smoke, rendererClasses [&NiagaraSpriteRendererProperties]}
  {emitterName=VelocitySmoke, rendererClasses [&NiagaraSpriteRendererProperties]}
 ]
 userVariables [
  User.MistColor {type=LinearColor, description="", defaultValue {struct=LinearColor, value=<rgba 0.0493 0.553 1 1>}}
  User.SkeletalMesh {
   type=NiagaraDataInterfaceSkeletalMesh
   defaultValue {
    struct=NiagaraExt_VariableValue_DataInterface
    value {dataInterface=&self:NiagaraDataInterfaceSkeletalMesh_0}
   }
  }
 ]
}
