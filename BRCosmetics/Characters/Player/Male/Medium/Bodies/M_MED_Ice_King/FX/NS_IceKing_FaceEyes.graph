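# Niagara cosmetic effect for the Ice King player body that adds mesh eyes and a drifting mist at the face.
# Eyes burst-spawns oriented mesh particles that are not killed by lifetime while the emitter is active.
# Face_Mist emits short-lived dark sprites that grow, fade, move along Y and sink under drag, with two spawn groups playing different SubUV frame ranges.
# An EyeColor user parameter defaults to bright blue.
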NiagaraSystem /BRCosmetics/Characters/Player/Male/Medium/Bodies/M_MED_Ice_King/FX/NS_IceKing_FaceEyes
dependencies {
 usedDataInterfaces [&NiagaraDataInterfaceCurve, &NiagaraDataInterfaceSpriteRendererInfo]
 usedDynamicInputs []
 usedModules [&SystemState]
}
emitters {
 Eyes {
  emitterUpdateScript [
   &EmitterState(
    Inactive Response=::NewEnumerator1
    Life Cycle Mode=::NewEnumerator1
    Loop Behavior=::NewEnumerator1
    Loop Duration Mode=::NewEnumerator1
    UseLoopDelay=false
   )
   &SpawnBurst_Instantaneous(Use Spawn Probability=false, UseLoopCountLimit=false)
  ]
  particleSpawnScript [
   &InitializeParticle(Color Mode=::NewEnumerator0, Lifetime=0, Mesh Scale Mode=::NewEnumerator3, UsePositionOffset=true, Write Material Random=true)
   &InitialMeshOrientation(Mesh Orientation Mode=::NewEnumerator2, Rotation=<0 0.25 0>, Use Rotation Vector=true)
  ]
  particleUpdateScript [&ParticleState(Kill Particles When Lifetime Has Elapsed=false, Let Infinitely Lived Particles Die When Emitter Deactivates=true)]
  rendererClasses [&NiagaraMeshRendererProperties]
  renderers [{rendererClass=&NiagaraMeshRendererProperties, rendererIndex=0}]
 }
 Face_Mist {
  emitterUpdateScript [
   &EmitterState()
   &SpawnRate(SpawnGroup=1, SpawnRate=&RandomRangeFloat(Minimum=5, Maximum=10), Use Spawn Probability=false) {module=SpawnRate001}
   &SpawnRate(SpawnRate=&RandomRangeFloat(Minimum=5, Maximum=10), Use Spawn Probability=false)
  ]
  particleSpawnScript [
   &InitializeParticle(
    Color=&MakeLinearColorFromVectorAndFloat("Vector (RGB)"=<0.009 0.009 0.009>, "Float (Alpha)"=&RandomRangeFloat(Minimum=0.25, Maximum=1))
    Lifetime Max=1.1
    Lifetime Min=0.85
    Lifetime Mode=::NewEnumerator1
    Sprite Rotation Mode=::NewEnumerator2
    Sprite Size Mode=::NewEnumerator1
    Uniform Sprite Size Max=35
    Uniform Sprite Size Min=20
    UsePositionOffset=false
    Write Material Random=true
   )
   &DynamicMaterialParameters(
    Index 0 Param 1=&RandomRangeFloat(Minimum=0, Maximum=1)
    Index 0 Param 2=&RandomRangeFloat(Minimum=0, Maximum=1)
    Module.Param0WriteEnabled=true
    Module.Param1WriteEnabled=true
    Module.Param2WriteEnabled=true
    Module.Param3WriteEnabled=false
   )
   &ShapeLocation(
    MaskSpawnGroup=false
    Offset=<-3 0 0>
    Offset Coordinate Space=::Local
    Offset Mode=::NewEnumerator0
    Override Seed=false
    Sphere Radius=5
    UseOwnerScale=true
   )
   &AddVelocity(Velocity=&RandomRangeVector(Minimum=<0 20 0>, Maximum=<0 60 0>))
  ]
  particleUpdateScript [
   &ParticleState()
   &ScaleSpriteSize(Uniform Curve Sprite Scale=DataInterface() {propertyValues {Curve {keys [<at 0 0.2 RCIM_Cubic>,<at 1 1.5 RCIM_Cubic>]}}})
   &SubUVAnimation(
    Direct Frame Index:NiagaraFloat=&MaskFloatBySpawnGroup(
     Masked Float=&FloatFromCurve(FloatCurve=DataInterface() {propertyValues {Curve {keys [
          {arriveTangent=55.9848, interpMode=RCIM_Cubic, leaveTangent=55.9848, tangentMode=RCTM_User, time=0, value=14}
          <at 0.4989 30.0046 RCIM_Cubic>
         ]}}})
     Passthrough Float=&FloatFromCurve(FloatCurve=DataInterface() {propertyValues {Curve {keys [{arriveTangent=55.9848, interpMode=RCIM_Cubic, leaveTangent=55.9848, tangentMode=RCTM_User, time=0, value=22}, <at 0.4989 36.754 RCIM_Cubic>]}}})
    )
    "Sprite Renderer "=DataInterface() {propertyValues {SpriteRenderer=&self:Face_Mist_0.NiagaraSpriteRendererProperties_0}}
    SubUV Animation Mode:ENiagara_SubUVLookupModeV2=::NewEnumerator5
   )
   &AccelerationForce(Acceleration=&RandomRangeVector(Minimum=<0 0 -70>, Maximum=<0 0 -100>))
   &Drag(Drag=&RandomRangeFloat(Minimum=1.5, Maximum=3), Use Linear Drag=true, Use Rotational Drag=true)
   &SolveForcesAndVelocity(Clamp Velocity=false, Limit Acceleration=false)
   &ScaleColor(
    Scale Alpha=&FloatFromCurve(FloatCurve=DataInterface() {propertyValues {Curve {keys [
         {interpMode=RCIM_Cubic, tangentMode=RCTM_User, time=0, value=0}
         {interpMode=RCIM_Cubic, tangentMode=RCTM_User, time=0.5, value=1}
         {tangentMode=RCTM_User, time=1, value=0}
        ]}}})
    ScaleA=true
    ScaleRGB=true
   )
  ]
  rendererClasses [&NiagaraSpriteRendererProperties]
  renderers [{rendererClass=&NiagaraSpriteRendererProperties}]
 }
}
summary {
 systemName=NS_IceKing_FaceEyes
 emitters [
  {emitterName=Face_Mist, rendererClasses [&NiagaraSpriteRendererProperties]}
  {emitterName=Eyes, rendererClasses [&NiagaraMeshRendererProperties]}
 ]
 userVariables [User.EyeColor {type=LinearColor, description="", defaultValue {struct=LinearColor, value=<rgba 0.136 0.3224 2 1>}}]
}
