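# Physics asset for the male medium Ice King body cosmetic, built for secondary motion of its hip cloths, loincloths and arm straps.
# Cloth strips are chains of box bodies with limited swing and tight twist, hung from tiny free-jointed base anchors, with arm-strap capsules on the upper arms.
# Capsules on torso, legs, feet and extra center thigh/calf collider bones, plus a large flat root box as floor, provide collision shapes.
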PhysicsAsset /BRCosmetics/Characters/Player/Male/Medium/Bodies/M_MED_Ice_King/Meshes/M_MED_IceKing_Body_Physics_New
bodies {
 Calf_C_Collider [{shapeType=Capsule, length=32.6582, radius=10.3701, shapeName=Calf_C_Collider_capsule, center=<-23.6283 0 0>, rotation=<rot 0 -90 178.4957>}]
 base_hipCloth_01_l [{shapeType=Capsule, length=0.505, radius=0.505, shapeName=base_hipCloth_01_l_capsule, rotation=<rot 0 -90 0>}]
 base_hipCloth_01_r [{shapeType=Capsule, length=0.505, radius=0.505, shapeName=base_hipCloth_01_r_capsule, rotation=<rot 0 -90 0>}]
 base_hipCloth_02_l [{shapeType=Capsule, length=0.505, radius=0.505, shapeName=base_hipCloth_02_l_capsule, rotation=<rot 0 -90 0>}]
 base_hipCloth_02_r [{shapeType=Capsule, length=0.505, radius=0.505, shapeName=base_hipCloth_02_r_capsule, rotation=<rot 0 -90 0>}]
 base_loinClothFront_01 [{shapeType=Capsule, length=0.505, radius=0.505, shapeName=base_loinClothFront_01_capsule, rotation=<rot 0 -90 0>}]
 base_loinClothRear_01 [{shapeType=Capsule, length=0.505, radius=0.505, shapeName=base_loinClothRear_01_capsule, rotation=<rot 0 -90 0>}]
 calf_l [
  {shapeType=Capsule, length=22.1236, radius=8.3403, shapeName=thigh_l_capsule, center=<-8.7896 1.0875 -1.5232>, rotation=<rot 0 -85.1376 -90>}
  {
   shapeType=Capsule
   length=22.1236
   radius=8.3403
   shapeName=thigh_l_capsule
   center=<-10.5855 -4.5878 -1.5232>
   rotation=<rot 0 -85.1376 -90>
  }
  {
   shapeType=Capsule
   length=16.4898
   radius=8.3403
   shapeName=thigh_l_capsule
   center=<-2.5701 6.2039 -0.8891>
   rotation=<rot -0.896 -71.2324 -90.7996>
  }
 ]
 calf_r [
  {shapeType=Capsule, length=22.1236, radius=8.3403, shapeName=thigh_r_capsule, center=<8.7896 -1.0875 1.5232>, rotation=<rot 0 -85.1376 90>}
  {shapeType=Capsule, length=22.1236, radius=8.3403, shapeName=thigh_r_capsule, center=<10.5855 4.5878 1.5232>, rotation=<rot 0 -85.1376 90>}
  {
   shapeType=Capsule
   length=16.4898
   radius=8.3403
   shapeName=thigh_r_capsule
   center=<2.5701 -6.2039 0.8891>
   rotation=<rot -0.896 -71.2324 89.2004>
  }
 ]
 calf_twist_02_l [{
   shapeType=Capsule
   length=12.6455
   radius=6.7476
   shapeName=thigh_l_capsule
   center=<-14.6245 -0.2001 -0.6531>
   rotation=<rot 0.5614 -101.7743 -87.2927>
  }]
 calf_twist_02_r [{
   shapeType=Capsule
   length=12.6455
   radius=6.7476
   shapeName=thigh_r_capsule
   center=<14.6245 0.2001 0.6531>
   rotation=<rot 0.5614 -101.7743 92.7073>
  }]
 dyn_armstrap_l [{shapeType=Capsule, length=2.186, radius=1.5718, center=<2.1505 0 0>, rotation=<rot 90 180 0>}]
 dyn_armstrap_r [{shapeType=Capsule, length=2.186, radius=1.5718, center=<2 0 0>, rotation=<rot 90 0 0>}]
 dyn_hipCloth_01_l [{shapeType=Box, extentX=8.9018, extentY=2.5, extentZ=15, center=<9.4366 0 0.5645>, rotation=<rot 0 0 -90>}]
 dyn_hipCloth_01_r [{shapeType=Box, extentX=10.2147, extentY=2.5, extentZ=15, center=<-9.3629 0 -0.5645>, rotation=<rot 0 0 90>}]
 dyn_hipCloth_02_l [{shapeType=Box, extentX=14, extentY=2.5, extentZ=15, center=<6.9949 0 0.0393>, rotation=<rot -3.7653 0.0001 90>}]
 dyn_hipCloth_02_r [{shapeType=Box, extentX=14, extentY=2.5, extentZ=15, center=<-7.0151 0 -0.3442>, rotation=<rot -5.0892 0.0002 -90.0001>}]
 dyn_hipCloth_03_l [{shapeType=Box, extentX=14, extentY=2.5, extentZ=15, center=<7 0 0>, rotation=<rot 0 0 90>}]
 dyn_hipCloth_03_r [{shapeType=Box, extentX=14, extentY=2.5, extentZ=15, center=<-7 0 0>, rotation=<rot 0 0 -90>}]
 dyn_hipCloth_04_l [{shapeType=Box, extentX=17.3517, extentY=3.6691, extentZ=15, center=<7.887 0 0>, rotation=<rot 0 0 89.8007>}]
 dyn_hipCloth_04_r [{shapeType=Box, extentX=17.3517, extentY=3.6691, extentZ=15, center=<-7.887 0 0>, rotation=<rot 0 0 -90.1993>}]
 dyn_hipCloth_05_l [{shapeType=Box, extentX=16.3288, extentY=4.0497, extentZ=15, center=<6.4988 0 0>, rotation=<rot 0 0 89.8007>}]
 dyn_hipCloth_05_r [{shapeType=Box, extentX=16.3288, extentY=4.0497, extentZ=15, center=<-6.4988 0 0>, rotation=<rot 0 0 -90.1993>}]
 dyn_hipCloth_06_l [{shapeType=Box, extentX=14, extentY=4.2011, extentZ=15, center=<7 0 0>, rotation=<rot 0 0 89.8007>}]
 dyn_hipCloth_06_r [{shapeType=Box, extentX=14, extentY=4.2011, extentZ=15, center=<-7 0 0>, rotation=<rot 0 0 -90.1993>}]
 dyn_hipCloth_07_l [{shapeType=Box, extentX=11.5756, extentY=4.3761, extentZ=15, center=<5.8956 0 0>, rotation=<rot 0 0 89.8007>}]
 dyn_hipCloth_07_r [{shapeType=Box, extentX=11.5756, extentY=4.3761, extentZ=15, center=<-5.8956 0 0>, rotation=<rot 0 0 -90.1993>}]
 dyn_loinClothFront_01 [{shapeType=Box, extentX=13.9398, extentY=4.5748, extentZ=10.3088, center=<7 0 0>, rotation=<rot 0 0 90>}]
 dyn_loinClothFront_02 [{shapeType=Box, extentX=13.9398, extentY=3.9399, extentZ=10.3088, center=<7 0 0>, rotation=<rot 0 0 90>}]
 dyn_loinClothFront_03 [{shapeType=Box, extentX=13.9398, extentY=4.6397, extentZ=10.6378, center=<7 0 0>, rotation=<rot 0 0 90>}]
 dyn_loinClothFront_04 [{shapeType=Box, extentX=15.9819, extentY=4.6397, extentZ=11.8975, center=<6.2036 0 0>, rotation=<rot 0 0 89.8007>}]
 dyn_loinClothFront_05 [{shapeType=Box, extentX=13.9398, extentY=4.6397, extentZ=13.6538, center=<7 0 0>, rotation=<rot 0 0 89.8007>}]
 dyn_loinClothFront_06 [{shapeType=Box, extentX=13.9398, extentY=4.6397, extentZ=13.6824, center=<7 0 0>, rotation=<rot 0 0 89.8007>}]
 dyn_loinClothFront_07 [{shapeType=Box, extentX=7.801, extentY=4.6397, extentZ=14.9779, center=<7 0 0>, rotation=<rot 0 0 89.8007>}]
 dyn_loinClothRear_01 [{shapeType=Box, extentX=12.1975, extentY=2.5, extentZ=15, center=<6.2166 0 -0.5349>, rotation=<rot 0 0 90>}]
 dyn_loinClothRear_02 [{shapeType=Box, extentX=17.4076, extentY=2.5, extentZ=15, center=<7.6708 0 -0.6499>, rotation=<rot 0 0 90>}]
 dyn_loinClothRear_03 [{shapeType=Box, extentX=15.1804, extentY=2.5, extentZ=15, center=<6.7193 0 -0.7774>, rotation=<rot 0 0 89.8007>}]
 dyn_loinClothRear_04 [{shapeType=Box, extentX=15.2399, extentY=2.5, extentZ=15, center=<6.5242 0 0>, rotation=<rot 0 0 89.8007>}]
 dyn_loinClothRear_05 [{shapeType=Box, extentX=17.3789, extentY=2.5, extentZ=15, center=<8.1325 0 0>, rotation=<rot 0 0 89.8007>}]
 dyn_loinClothRear_06 [{shapeType=Box, extentX=16.4173, extentY=2.5, extentZ=15.6065, center=<7.2563 0 0>, rotation=<rot 0 0 89.8007>}]
 foot_l [{
   shapeType=Capsule
   length=21.4489
   radius=8.0952
   shapeName=foot_l_capsule
   center=<-4.1757 8.2758 -0.1132>
   rotation=<rot 0 -168.4454 -89.3078>
  }]
 foot_r [{
   shapeType=Capsule
   length=21.4489
   radius=8.0952
   shapeName=foot_r_capsule
   center=<4.1757 -8.2758 0.1132>
   rotation=<rot 0 -168.4454 90.6922>
  }]
 pelvis [{shapeType=Capsule, length=4.4873, radius=12.5973, center=<-1.3872 0.9574 0>}]
 root [{shapeType=Box, extentX=400, extentY=400, extentZ=8, shapeName=root_box, center=<0 0 -3>}]
 spine_01 [{shapeType=Capsule, length=0, radius=13.0143, center=<-1.3872 0.8581 0>}]
 spine_03 [
  {
   shapeType=Capsule
   length=16.0943
   radius=24.0436
   shapeName=spine_03_capsule
   center=<11.7279 0.5265 0.0001>
   rotation=<rot -0.0003 -90 -0.0004>
  }
  {shapeType=Capsule, length=15, radius=27.766, shapeName=spine_03_capsule, center=<33.6433 20.1151 0>}
 ]
 thigh_C_Collider [{shapeType=Capsule, length=35.2422, radius=7.9508, shapeName=Calf_C_Collider_capsule, center=<-19.3727 0 0>, rotation=<rot 0 -90 90>}]
 thigh_l [{shapeType=Capsule, length=10.8236, radius=8.5695, shapeName=thigh_l_capsule, center=<-28.4675 2.1901 -0.981>, rotation=<rot 0 -85.1376 -90>}]
 thigh_r [{shapeType=Capsule, length=10.8236, radius=8.5695, shapeName=thigh_r_capsule, center=<28.4675 -2.1901 0.981>, rotation=<rot 0 -85.1376 90>}]
 thigh_twist_01_l [
  {
   shapeType=Capsule
   length=17.9429
   radius=8.7587
   shapeName=thigh_twist_01_l_capsule
   center=<9.3931 1.0941 -0.6693>
   rotation=<rot 0 -95.9146 -91.3826>
  }
  {
   shapeType=Capsule
   length=15.4458
   radius=9.2181
   shapeName=thigh_twist_01_l_capsule
   center=<9.4432 5.2025 -0.7895>
   rotation=<rot 0 -90.0002 -93.0481>
  }
 ]
 thigh_twist_01_r [
  {
   shapeType=Capsule
   length=17.9429
   radius=8.7587
   shapeName=thigh_twist_01_r_capsule
   center=<-9.3931 -1.0941 0.6693>
   rotation=<rot 0 -95.9146 88.6174>
  }
  {
   shapeType=Capsule
   length=15.4458
   radius=9.2181
   shapeName=thigh_twist_01_r_capsule
   center=<-9.439 -5.2025 0.7113>
   rotation=<rot 0 -90.0001 88.3142>
  }
 ]
 upperarm_l [{shapeType=Capsule, length=0, radius=1, rotation=<rot 0 0 -180>}]
 upperarm_r [{shapeType=Capsule, length=0, radius=1}]
}
constraints [
 {
  bone1Name=base_hipCloth_01_l
  bone2Name=pelvis
  swing1LimitDegrees=45
  swing1Motion=Free
  swing2LimitDegrees=45
  swing2Motion=Free
  twistLimitDegrees=45
  twistMotion=Free
 }
 {
  bone1Name=base_hipCloth_02_l
  bone2Name=dyn_hipCloth_01_l
  swing1LimitDegrees=45
  swing1Motion=Free
  swing2LimitDegrees=45
  swing2Motion=Free
  twistLimitDegrees=45
  twistMotion=Free
 }
 {
  bone1Name=base_hipCloth_01_r
  bone2Name=pelvis
  swing1LimitDegrees=45
  swing1Motion=Free
  swing2LimitDegrees=45
  swing2Motion=Free
  twistLimitDegrees=45
  twistMotion=Free
 }
 {bone1Name=dyn_hipCloth_01_r, bone2Name=base_hipCloth_01_r, swing1LimitDegrees=20, swing2LimitDegrees=45, twistLimitDegrees=5}
 {
  bone1Name=base_hipCloth_02_r
  bone2Name=dyn_hipCloth_01_r
  swing1LimitDegrees=45
  swing1Motion=Free
  swing2LimitDegrees=45
  swing2Motion=Free
  twistLimitDegrees=45
  twistMotion=Free
 }
 {bone1Name=dyn_hipCloth_02_r, bone2Name=base_hipCloth_02_r, swing1LimitDegrees=20, swing2LimitDegrees=35, twistLimitDegrees=5}
 {bone1Name=dyn_hipCloth_03_r, bone2Name=dyn_hipCloth_02_r, swing1LimitDegrees=20, swing2LimitDegrees=35, twistLimitDegrees=5}
 {bone1Name=dyn_hipCloth_04_r, bone2Name=dyn_hipCloth_03_r, swing1LimitDegrees=20, swing2LimitDegrees=35, twistLimitDegrees=5}
 {bone1Name=dyn_hipCloth_05_r, bone2Name=dyn_hipCloth_04_r, swing1LimitDegrees=20, swing2LimitDegrees=35, twistLimitDegrees=5}
 {bone1Name=dyn_hipCloth_06_r, bone2Name=dyn_hipCloth_05_r, swing1LimitDegrees=20, swing2LimitDegrees=35, twistLimitDegrees=5}
 {bone1Name=dyn_hipCloth_07_r, bone2Name=dyn_hipCloth_06_r, swing1LimitDegrees=20, swing2LimitDegrees=35, twistLimitDegrees=5}
 {
  bone1Name=base_loinClothFront_01
  bone2Name=pelvis
  swing1LimitDegrees=45
  swing1Motion=Free
  swing2LimitDegrees=45
  swing2Motion=Free
  twistLimitDegrees=45
  twistMotion=Free
 }
 {bone1Name=dyn_loinClothFront_01, bone2Name=base_loinClothFront_01, swing1LimitDegrees=10, swing2LimitDegrees=15, twistLimitDegrees=5}
 {bone1Name=dyn_loinClothFront_02, bone2Name=dyn_loinClothFront_01, swing1LimitDegrees=10, swing2LimitDegrees=25, twistLimitDegrees=5}
 {bone1Name=dyn_loinClothFront_03, bone2Name=dyn_loinClothFront_02, swing1LimitDegrees=10, swing2LimitDegrees=20, twistLimitDegrees=5}
 {bone1Name=dyn_loinClothFront_04, bone2Name=dyn_loinClothFront_03, swing1LimitDegrees=10, swing2LimitDegrees=15, twistLimitDegrees=5}
 {bone1Name=dyn_loinClothFront_05, bone2Name=dyn_loinClothFront_04, swing1LimitDegrees=10, swing2LimitDegrees=10, twistLimitDegrees=5}
 {bone1Name=dyn_loinClothFront_06, bone2Name=dyn_loinClothFront_05, swing1LimitDegrees=10, swing2LimitDegrees=10, twistLimitDegrees=5}
 {bone1Name=dyn_loinClothFront_07, bone2Name=dyn_loinClothFront_06, swing1LimitDegrees=10, swing2LimitDegrees=10, twistLimitDegrees=5}
 {
  bone1Name=base_loinClothRear_01
  bone2Name=pelvis
  swing1LimitDegrees=45
  swing1Motion=Free
  swing2LimitDegrees=45
  swing2Motion=Free
  twistLimitDegrees=45
  twistMotion=Free
 }
 {bone1Name=dyn_loinClothRear_01, bone2Name=base_loinClothRear_01, swing1LimitDegrees=20, swing2LimitDegrees=15, twistLimitDegrees=5}
 {bone1Name=dyn_loinClothRear_02, bone2Name=dyn_loinClothRear_01, swing1LimitDegrees=20, swing2LimitDegrees=25, twistLimitDegrees=5}
 {bone1Name=dyn_loinClothRear_03, bone2Name=dyn_loinClothRear_02, swing1LimitDegrees=20, swing2LimitDegrees=30, twistLimitDegrees=5}
 {bone1Name=dyn_loinClothRear_04, bone2Name=dyn_loinClothRear_03, swing1LimitDegrees=20, swing2LimitDegrees=35, twistLimitDegrees=5}
 {bone1Name=dyn_loinClothRear_05, bone2Name=dyn_loinClothRear_04, swing1LimitDegrees=20, swing2LimitDegrees=45, twistLimitDegrees=5}
 {bone1Name=dyn_loinClothRear_06, bone2Name=dyn_loinClothRear_05, swing1LimitDegrees=20, swing2LimitDegrees=45, twistLimitDegrees=5}
 {
  bone1Name=spine_01
  bone2Name=pelvis
  swing1LimitDegrees=45
  swing1Motion=Free
  swing2LimitDegrees=45
  swing2Motion=Free
  twistLimitDegrees=45
  twistMotion=Free
 }
 {
  bone1Name=thigh_l
  bone2Name=pelvis
  swing1LimitDegrees=45
  swing1Motion=Free
  swing2LimitDegrees=45
  swing2Motion=Free
  twistLimitDegrees=45
  twistMotion=Free
 }
 {
  bone1Name=calf_l
  bone2Name=thigh_l
  swing1LimitDegrees=45
  swing1Motion=Free
  swing2LimitDegrees=45
  swing2Motion=Free
  twistLimitDegrees=45
  twistMotion=Free
 }
 {
  bone1Name=thigh_twist_01_l
  bone2Name=thigh_l
  swing1LimitDegrees=45
  swing1Motion=Free
  swing2LimitDegrees=45
  swing2Motion=Free
  twistLimitDegrees=45
  twistMotion=Free
 }
 {
  bone1Name=calf_twist_02_l
  bone2Name=calf_l
  swing1LimitDegrees=45
  swing1Motion=Free
  swing2LimitDegrees=45
  swing2Motion=Free
  twistLimitDegrees=45
  twistMotion=Free
 }
 {
  bone1Name=foot_l
  bone2Name=calf_l
  swing1LimitDegrees=45
  swing1Motion=Free
  swing2LimitDegrees=45
  swing2Motion=Free
  twistLimitDegrees=45
  twistMotion=Free
 }
 {
  bone1Name=thigh_r
  bone2Name=pelvis
  swing1LimitDegrees=45
  swing1Motion=Free
  swing2LimitDegrees=45
  swing2Motion=Free
  twistLimitDegrees=45
  twistMotion=Free
 }
 {
  bone1Name=calf_r
  bone2Name=thigh_r
  swing1LimitDegrees=45
  swing1Motion=Free
  swing2LimitDegrees=45
  swing2Motion=Free
  twistLimitDegrees=45
  twistMotion=Free
 }
 {
  bone1Name=thigh_twist_01_r
  bone2Name=thigh_r
  swing1LimitDegrees=45
  swing1Motion=Free
  swing2LimitDegrees=45
  swing2Motion=Free
  twistLimitDegrees=45
  twistMotion=Free
 }
 {
  bone1Name=calf_twist_02_r
  bone2Name=calf_r
  swing1LimitDegrees=45
  swing1Motion=Free
  swing2LimitDegrees=45
  swing2Motion=Free
  twistLimitDegrees=45
  twistMotion=Free
 }
 {
  bone1Name=foot_r
  bone2Name=calf_r
  swing1LimitDegrees=45
  swing1Motion=Free
  swing2LimitDegrees=45
  swing2Motion=Free
  twistLimitDegrees=45
  twistMotion=Free
 }
 {bone1Name=dyn_hipCloth_01_l, bone2Name=base_hipCloth_01_l, swing1LimitDegrees=20, swing2LimitDegrees=45, twistLimitDegrees=5}
 {bone1Name=dyn_hipCloth_02_l, bone2Name=base_hipCloth_02_l, swing1LimitDegrees=20, swing2LimitDegrees=45, twistLimitDegrees=5}
 {bone1Name=dyn_hipCloth_03_l, bone2Name=dyn_hipCloth_02_l, swing1LimitDegrees=20, swing2LimitDegrees=35, twistLimitDegrees=5}
 {bone1Name=dyn_hipCloth_04_l, bone2Name=dyn_hipCloth_03_l, swing1LimitDegrees=20, swing2LimitDegrees=35, twistLimitDegrees=5}
 {bone1Name=dyn_hipCloth_05_l, bone2Name=dyn_hipCloth_04_l, swing1LimitDegrees=20, swing2LimitDegrees=35, twistLimitDegrees=5}
 {bone1Name=dyn_hipCloth_06_l, bone2Name=dyn_hipCloth_05_l, swing1LimitDegrees=20, swing2LimitDegrees=35, twistLimitDegrees=5}
 {bone1Name=dyn_hipCloth_07_l, bone2Name=dyn_hipCloth_06_l, swing1LimitDegrees=20, swing2LimitDegrees=35, twistLimitDegrees=5}
 {bone1Name=dyn_armstrap_r, bone2Name=upperarm_r, swing1LimitDegrees=5, swing2LimitDegrees=30, twistLimitDegrees=5}
 {
  bone1Name=upperarm_r
  bone2Name=spine_03
  swing1LimitDegrees=45
  swing1Motion=Free
  swing2LimitDegrees=45
  swing2Motion=Free
  twistLimitDegrees=45
  twistMotion=Free
 }
 {
  bone1Name=upperarm_l
  bone2Name=spine_03
  swing1LimitDegrees=45
  swing1Motion=Free
  swing2LimitDegrees=45
  swing2Motion=Free
  twistLimitDegrees=45
  twistMotion=Free
 }
 {bone1Name=dyn_armstrap_l, bone2Name=upperarm_l, swing1LimitDegrees=5, swing2LimitDegrees=30, twistLimitDegrees=5}
 {
  bone1Name=pelvis
  bone2Name=root
  swing1LimitDegrees=45
  swing1Motion=Free
  swing2LimitDegrees=45
  swing2Motion=Free
  twistLimitDegrees=45
  twistMotion=Free
 }
 {
  bone1Name=Calf_C_Collider
  bone2Name=root
  swing1LimitDegrees=45
  swing1Motion=Free
  swing2LimitDegrees=45
  swing2Motion=Free
  twistLimitDegrees=45
  twistMotion=Free
 }
 {
  bone1Name=thigh_C_Collider
  bone2Name=pelvis
  swing1LimitDegrees=45
  swing1Motion=Free
  swing2LimitDegrees=45
  swing2Motion=Free
  twistLimitDegrees=45
  twistMotion=Free
 }
 {
  bone1Name=spine_03
  bone2Name=spine_01
  swing1LimitDegrees=45
  swing1Motion=Free
  swing2LimitDegrees=45
  swing2Motion=Free
  twistLimitDegrees=45
  twistMotion=Free
 }
]
