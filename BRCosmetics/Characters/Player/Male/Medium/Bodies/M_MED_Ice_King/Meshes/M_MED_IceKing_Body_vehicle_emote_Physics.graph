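# Physics asset for the Ice King male body, apparently its vehicle emote variant, covering dangling hip cloth, loincloths and arm straps.
# Box-body chains for left and right hip cloth and front and rear loincloths hang from the pelvis with angular limits.
# Leg capsules provide body volume, and tiny upper-arm bodies anchor the straps.
# Body-to-body joints are free, except left thigh to pelvis, which lacks the right side's free motion.
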PhysicsAsset /BRCosmetics/Characters/Player/Male/Medium/Bodies/M_MED_Ice_King/Meshes/M_MED_IceKing_Body_vehicle_emote_Physics
bodies {
 calf_l [
  {shapeType=Capsule, length=26.9243, radius=8.1998, center=<-18.8709 -2.5256 0.1052>, rotation=<rot 0 88.6315 -89.2119>}
  {shapeType=Capsule, length=25.7334, radius=9.3907, center=<-7.7338 0.422 0.3948>, rotation=<rot -0.394 118.6292 -89.3175>}
 ]
 calf_r [
  {shapeType=Capsule, length=26.9243, radius=8.1998, center=<18.8709 2.5256 0.1052>, rotation=<rot 0 -91.3685 -89.2119>}
  {shapeType=Capsule, length=25.7334, radius=9.3907, center=<7.7338 -0.422 0.3948>, rotation=<rot -0.394 -61.3708 -89.3175>}
 ]
 dyn_armstrap_l [{shapeType=Capsule, length=2.186, radius=1.5718, center=<2 0 0>, rotation=<rot 90 0 0>}]
 dyn_armstrap_r [{shapeType=Capsule, length=2.5789, radius=1.6264, center=<2 0 0>, rotation=<rot 90 0 0>}]
 dyn_hipCloth_01_l [{shapeType=Box, extentX=14, extentY=2.5, extentZ=15, center=<7 0 0>, rotation=<rot 0 0 90>}]
 dyn_hipCloth_01_r [{shapeType=Box, extentX=14, extentY=2.5, extentZ=15, center=<-7 0 0>, rotation=<rot 0 0 90>}]
 dyn_hipCloth_02_l [{shapeType=Box, extentX=14, extentY=2.5, extentZ=15, center=<7 0 0>, rotation=<rot 0 0 90>}]
 dyn_hipCloth_02_r [{shapeType=Box, extentX=14, extentY=2.5, extentZ=15, center=<-7 0 0>, rotation=<rot 0 0 90>}]
 dyn_hipCloth_03_l [{shapeType=Box, extentX=14, extentY=2.5, extentZ=15, center=<7 0 0>, rotation=<rot 0 0 90>}]
 dyn_hipCloth_03_r [{shapeType=Box, extentX=14, extentY=2.5, extentZ=15, center=<-7 0 0>, rotation=<rot 0 0 90>}]
 dyn_hipCloth_04_l [{shapeType=Box, extentX=14, extentY=4.8177, extentZ=15, center=<7 0 0>, rotation=<rot 0 0 90>}]
 dyn_hipCloth_04_r [{shapeType=Box, extentX=14, extentY=4.8177, extentZ=15, center=<-7 0 0>, rotation=<rot 0 0 90>}]
 dyn_hipCloth_05_l [{shapeType=Box, extentX=14, extentY=4.8177, extentZ=15, center=<7 0 0>, rotation=<rot 0 0 90>}]
 dyn_hipCloth_05_r [{shapeType=Box, extentX=14, extentY=4.8177, extentZ=15, center=<-7 0 0>, rotation=<rot 0 0 90>}]
 dyn_hipCloth_06_l [{shapeType=Box, extentX=14, extentY=4.8177, extentZ=15, center=<7 0 0>, rotation=<rot 0 0 90>}]
 dyn_hipCloth_06_r [{shapeType=Box, extentX=14, extentY=4.8177, extentZ=15, center=<-7 0 0>, rotation=<rot 0 0 90>}]
 dyn_hipCloth_07_l [{shapeType=Box, extentX=14, extentY=4.8177, extentZ=15, center=<7 0 0>, rotation=<rot 0 0 90>}]
 dyn_hipCloth_07_r [{shapeType=Box, extentX=14, extentY=4.8177, extentZ=15, center=<-7 0 0>, rotation=<rot 0 0 90>}]
 dyn_loinClothFront_01 [{shapeType=Box, extentX=13.9398, extentY=2.4398, extentZ=10.3088, center=<7 0 0>, rotation=<rot 0 0 90>}]
 dyn_loinClothFront_02 [{shapeType=Box, extentX=13.9398, extentY=3.9399, extentZ=10.3088, center=<7 0 0>, rotation=<rot 0 0 90>}]
 dyn_loinClothFront_03 [{shapeType=Box, extentX=13.9398, extentY=4.6397, extentZ=10.6378, center=<7 0 0>, rotation=<rot 0 0 90>}]
 dyn_loinClothFront_04 [{shapeType=Box, extentX=13.9398, extentY=4.6397, extentZ=10.4579, center=<7 0 0>, rotation=<rot 0 0 90>}]
 dyn_loinClothFront_05 [{shapeType=Box, extentX=13.9398, extentY=4.6397, extentZ=10.6378, center=<7 0 0>, rotation=<rot 0 0 90>}]
 dyn_loinClothFront_06 [{shapeType=Box, extentX=13.9398, extentY=4.6397, extentZ=10.6378, center=<7 0 0>, rotation=<rot 0 0 90>}]
 dyn_loinClothFront_07 [{shapeType=Box, extentX=7.801, extentY=4.6397, extentZ=10.6378, center=<7 0 0>, rotation=<rot 0 0 90>}]
 dyn_loinClothRear_01 [{shapeType=Box, extentX=14, extentY=2.5, extentZ=15, center=<7 0 0>, rotation=<rot 0 0 90>}]
 dyn_loinClothRear_02 [{shapeType=Box, extentX=14, extentY=2.5, extentZ=15, center=<7 0 0>, rotation=<rot 0 0 90>}]
 dyn_loinClothRear_03 [{shapeType=Box, extentX=14, extentY=2.5, extentZ=15, center=<7 0 0>, rotation=<rot 0 0 90>}]
 dyn_loinClothRear_04 [{shapeType=Box, extentX=14, extentY=2.5, extentZ=15, center=<7 0 0>, rotation=<rot 0 0 90>}]
 dyn_loinClothRear_05 [{shapeType=Box, extentX=14, extentY=2.5, extentZ=15, center=<7 0 0>, rotation=<rot 0 0 90>}]
 dyn_loinClothRear_06 [{shapeType=Box, extentX=14, extentY=2.5, extentZ=15, center=<7 0 0>, rotation=<rot 0 0 90>}]
 pelvis [{shapeType=Capsule, length=8.0539, radius=10.1522}]
 thigh_l [
  {shapeType=Capsule, length=27.39, radius=8.0792, center=<-17.1977 0.0106 -0.2525>, rotation=<rot 90 0 0.0001>}
  {shapeType=Capsule, length=7.8879, radius=10.0581, center=<-14.2197 3.0948 0.7751>, rotation=<rot 84.9923 179.9998 -180>}
 ]
 thigh_r [
  {shapeType=Capsule, length=27.39, radius=10.0792, center=<17.1977 -0.0106 -0.2525>, rotation=<rot 90 0 -179.9999>}
  {shapeType=Capsule, length=7.8879, radius=11.1831, center=<14.1081 -3.0947 -0.4989>, rotation=<rot 84.9928 -0.0002 -180>}
 ]
 upperarm_l [{shapeType=Capsule, length=0, radius=1}]
 upperarm_r [{shapeType=Capsule, length=0, radius=1}]
}
constraints [
 {
  bone1Name=calf_r
  bone2Name=thigh_r
  swing1LimitDegrees=45
  swing1Motion=Free
  swing2LimitDegrees=45
  swing2Motion=Free
  twistLimitDegrees=45
  twistMotion=Free
 }
 {
  bone1Name=thigh_r
  bone2Name=pelvis
  swing1LimitDegrees=45
  swing1Motion=Free
  swing2LimitDegrees=45
  swing2Motion=Free
  twistLimitDegrees=45
  twistMotion=Free
 }
 {bone1Name=dyn_loinClothRear_01, bone2Name=pelvis, swing1LimitDegrees=40, swing2LimitDegrees=40, twistLimitDegrees=5}
 {bone1Name=dyn_loinClothRear_02, bone2Name=dyn_loinClothRear_01, swing1LimitDegrees=40, swing2LimitDegrees=40, twistLimitDegrees=5}
 {bone1Name=dyn_loinClothRear_03, bone2Name=dyn_loinClothRear_02, swing1LimitDegrees=40, swing2LimitDegrees=40, twistLimitDegrees=5}
 {bone1Name=dyn_loinClothRear_04, bone2Name=dyn_loinClothRear_03, swing1LimitDegrees=40, swing2LimitDegrees=40, twistLimitDegrees=5}
 {bone1Name=dyn_loinClothRear_05, bone2Name=dyn_loinClothRear_04, swing1LimitDegrees=40, swing2LimitDegrees=40, twistLimitDegrees=5}
 {bone1Name=dyn_loinClothRear_06, bone2Name=dyn_loinClothRear_05, swing1LimitDegrees=40, swing2LimitDegrees=40, twistLimitDegrees=5}
 {bone1Name=dyn_hipCloth_01_l, bone2Name=pelvis, swing1LimitDegrees=40, swing2LimitDegrees=40, twistLimitDegrees=5}
 {bone1Name=dyn_hipCloth_02_l, bone2Name=dyn_hipCloth_01_l, swing1LimitDegrees=40, swing2LimitDegrees=40, twistLimitDegrees=10}
 {bone1Name=dyn_hipCloth_03_l, bone2Name=dyn_hipCloth_02_l, swing1LimitDegrees=40, swing2LimitDegrees=40, twistLimitDegrees=10}
 {bone1Name=dyn_hipCloth_04_l, bone2Name=dyn_hipCloth_03_l, swing1LimitDegrees=40, swing2LimitDegrees=40, twistLimitDegrees=10}
 {bone1Name=dyn_hipCloth_05_l, bone2Name=dyn_hipCloth_04_l, swing1LimitDegrees=40, swing2LimitDegrees=40, twistLimitDegrees=10}
 {bone1Name=dyn_hipCloth_06_l, bone2Name=dyn_hipCloth_05_l, swing1LimitDegrees=40, swing2LimitDegrees=40, twistLimitDegrees=10}
 {bone1Name=dyn_hipCloth_07_l, bone2Name=dyn_hipCloth_06_l, swing1LimitDegrees=40, swing2LimitDegrees=40, twistLimitDegrees=10}
 {bone1Name=dyn_loinClothFront_01, bone2Name=pelvis, swing1LimitDegrees=40, swing2LimitDegrees=40, twistLimitDegrees=5}
 {bone1Name=dyn_loinClothFront_02, bone2Name=dyn_loinClothFront_01, swing1LimitDegrees=40, swing2LimitDegrees=40, twistLimitDegrees=5}
 {bone1Name=dyn_loinClothFront_03, bone2Name=dyn_loinClothFront_02, swing1LimitDegrees=40, swing2LimitDegrees=40, twistLimitDegrees=5}
 {bone1Name=dyn_loinClothFront_04, bone2Name=dyn_loinClothFront_03, swing1LimitDegrees=40, swing2LimitDegrees=40, twistLimitDegrees=5}
 {bone1Name=dyn_loinClothFront_05, bone2Name=dyn_loinClothFront_04, swing1LimitDegrees=40, swing2LimitDegrees=40, twistLimitDegrees=5}
 {bone1Name=dyn_loinClothFront_06, bone2Name=dyn_loinClothFront_05, swing1LimitDegrees=40, swing2LimitDegrees=40, twistLimitDegrees=5}
 {bone1Name=dyn_loinClothFront_07, bone2Name=dyn_loinClothFront_06, swing1LimitDegrees=40, swing2LimitDegrees=40, twistLimitDegrees=5}
 {bone1Name=dyn_hipCloth_01_r, bone2Name=pelvis, swing1LimitDegrees=40, swing2LimitDegrees=40, twistLimitDegrees=5}
 {bone1Name=dyn_hipCloth_02_r, bone2Name=dyn_hipCloth_01_r, swing1LimitDegrees=40, swing2LimitDegrees=40, twistLimitDegrees=5}
 {bone1Name=dyn_hipCloth_03_r, bone2Name=dyn_hipCloth_02_r, swing1LimitDegrees=40, swing2LimitDegrees=40, twistLimitDegrees=5}
 {bone1Name=dyn_hipCloth_04_r, bone2Name=dyn_hipCloth_03_r, swing1LimitDegrees=40, swing2LimitDegrees=40, twistLimitDegrees=5}
 {bone1Name=dyn_hipCloth_05_r, bone2Name=dyn_hipCloth_04_r, swing1LimitDegrees=40, swing2LimitDegrees=40, twistLimitDegrees=5}
 {bone1Name=dyn_hipCloth_06_r, bone2Name=dyn_hipCloth_05_r, swing1LimitDegrees=40, swing2LimitDegrees=40, twistLimitDegrees=5}
 {bone1Name=dyn_hipCloth_07_r, bone2Name=dyn_hipCloth_06_r, swing1LimitDegrees=40, swing2LimitDegrees=40, twistLimitDegrees=5}
 {
  bone1Name=calf_l
  bone2Name=thigh_l
  swing1LimitDegrees=45
  swing1Motion=Free
  swing2LimitDegrees=45
  swing2Motion=Free
  twistLimitDegrees=45
  twistMotion=Free
 }
 {
  bone1Name=upperarm_l
  bone2Name=pelvis
  swing1LimitDegrees=45
  swing1Motion=Free
  swing2LimitDegrees=45
  swing2Motion=Free
  twistLimitDegrees=45
  twistMotion=Free
 }
 {bone1Name=dyn_armstrap_l, bone2Name=upperarm_l, swing1LimitDegrees=5, swing2LimitDegrees=30, twistLimitDegrees=5}
 {
  bone1Name=upperarm_r
  bone2Name=pelvis
  swing1LimitDegrees=45
  swing1Motion=Free
  swing2LimitDegrees=45
  swing2Motion=Free
  twistLimitDegrees=45
  twistMotion=Free
 }
 {bone1Name=dyn_armstrap_r, bone2Name=upperarm_r, swing1LimitDegrees=5, swing2LimitDegrees=30, twistLimitDegrees=5}
 {bone1Name=thigh_l, bone2Name=pelvis, swing1LimitDegrees=40, swing2LimitDegrees=40, twistLimitDegrees=5}
]
