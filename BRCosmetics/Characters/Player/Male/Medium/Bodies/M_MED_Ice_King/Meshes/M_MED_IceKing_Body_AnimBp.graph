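# Post-process style character-part AnimBP for the male medium Ice King body cosmetic, adding secondary motion to shoulder armor and cloth.
# Copied parent pose runs through RBF pose drivers and Control Rig into a cache.
# Bool blends select between two RigidBody setups, one for a vehicle emote, and a constraint plus trail-controller cloth chain, with additive bone offsets.
# Events update emote, front-end, movement and gravity state, and reset dynamics on character transitions.
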AnimBlueprint /BRCosmetics/Characters/Player/Male/Medium/Bodies/M_MED_Ice_King/Meshes/M_MED_IceKing_Body_AnimBp
target_skeleton=/Game/Characters/Player/Male/Medium/Bodies/M_MED_Ice_King/Meshes/M_MED_IceKing_Body_Skeleton
graphs [
 AnimGraph {nodes [
   RigidBody!(ComponentPose, Alpha, OverrideWorldGravity:Vector="0,0,0") -> (Pose) {
    bOverrideWorldGravity=true
    overridePhysicsAsset=&/M_MED_IceKing_Body_Physics_New
    simSpaceSettings {
     maxAngularAcceleration=1600
     maxAngularVelocity=4
     maxLinearAcceleration=1000
     maxLinearVelocity=300
     worldAlpha=1
     externalLinearDragV=<1.2 1.2 1.2>
    }
   }
   Output Pose(Result=Use cached pose 'DynamicsCache'()) {layerGroup=None}
   n1 = RigidBody(ComponentPose=Local To Component(LocalPose=Use cached pose 'prerigid'()), OverrideWorldGravity="0,0,-980") {
    bClampLinearTranslationLimitToRefPose=true
    bOverrideWorldGravity=true
    overridePhysicsAsset=&/M_MED_IceKing_Body_vehicle_emote_Physics
    simulationSpace=BaseBoneSpace
    alphaBoolBlend {blendOption=Linear}
    baseBoneRef {boneName=pelvis}
    componentAppliedLinearAccClamp=<1 1 1>
    overrideWorldGravity {z=-980}
   }
   n2 = Component To Local(ComponentPose=@n1)
   n3 = RigidBody(ComponentPose=Local To Component(LocalPose=Use cached pose 'prerigid'()), OverrideWorldGravity="0,0,0") {
    bOverrideWorldGravity=true
    overridePhysicsAsset=&M_MED_IceKing_Body_Physics_New
    simulationSpace=BaseBoneSpace
    baseBoneRef {boneName=root}
    simSpaceSettings {
     maxAngularAcceleration=1200
     maxAngularVelocity=4
     maxLinearAcceleration=900
     maxLinearVelocity=250
     velocityScaleZ=0.45
     worldAlpha=1
     externalLinearDragV=<1.2 1.2 1.2>
    }
   }
   n4 = "Transform (Modify) Bone - Bone: base_hipCloth_01_r"!(ComponentPose=@n3, bAlphaBoolEnabled:bool=True, Translation="0,0,0", Rotation="-2,0,0") -> (Pose) {
    alphaInputType=Bool
    rotationMode=BMM_Additive
    rotationSpace=BCS_BoneSpace
    translationSpace=BCS_BoneSpace
    alphaBoolBlend {blendInTime=0.1, blendOutTime=0.1}
    boneToModify {boneName=base_hipCloth_01_r}
    rotation {pitch=-2}
   }
   n5 = "Transform (Modify) Bone - Bone: dyn_hipCloth_03_l"!(ComponentPose=@n4, bAlphaBoolEnabled:bool=True, Translation="0,0,0", Rotation="-7,3,-10") -> (Pose) {
    alphaInputType=Bool
    rotationMode=BMM_Additive
    rotationSpace=BCS_BoneSpace
    translationSpace=BCS_BoneSpace
    alphaBoolBlend {blendInTime=0.1, blendOutTime=0.1}
    boneToModify {boneName=dyn_hipCloth_03_l}
    rotation=<rot -7 3 -10>
   }
   n6 = "Transform (Modify) Bone - Bone: dyn_hipCloth_05_l"!(ComponentPose=@n5, bAlphaBoolEnabled:bool=True, Translation="0,0,0", Rotation="7,0,0") -> (Pose) {
    alphaInputType=Bool
    rotationMode=BMM_Additive
    rotationSpace=BCS_BoneSpace
    translationSpace=BCS_BoneSpace
    alphaBoolBlend {blendInTime=0.1, blendOutTime=0.1}
    boneToModify {boneName=dyn_hipCloth_05_l}
    rotation {pitch=7}
   }
   n7 = Component To Local(ComponentPose=@n6)
   n8 = "Constraint - dyn_hipCloth_02_l (P)"(ComponentPose=Local To Component(LocalPose=Use cached pose 'CopyPoseCache'()), Alpha=1, ConstraintWeights_0=1) {
    alphaBoolBlend {blendOption=Linear}
    boneToModify {boneName=dyn_hipCloth_02_l}
    constraintSetup [{targetBone {boneName=thigh_l}}]
    constraintWeights [1]
   }
   n9 = "Constraint - dyn_hipCloth_02_r (P)"(ComponentPose=@n8, Alpha=1, ConstraintWeights_0=1) {
    alphaBoolBlend {blendOption=Linear}
    boneToModify {boneName=dyn_hipCloth_02_r}
    constraintSetup [{targetBone {boneName=thigh_r}}]
    constraintWeights [1]
   }
   n10 = Trail controller - Bone: dyn_hipCloth_07_l(ComponentPose=@n9, Alpha=1) {
    bLimitRotation=true
    chainLength=5
    relaxationSpeedScale=3
    rotationLimits [{limitMax=<180 10 180>}, {limitMax=<180 10 180>}, {limitMax=<180 10 180>}, {limitMax=<180 10 180>}, {limitMax=<180 0 180>}]
    rotationOffsets [<0 0 0>, <0 0 0>, <0 0 0>, <0 0 0>, <0 0 0>]
    trailBone {boneName=dyn_hipCloth_07_l}
   }
   n11 = Trail controller - Bone: dyn_hipCloth_07_r(ComponentPose=@n10, Alpha=1) {
    bInvertChainBoneAxis=true
    bLimitRotation=true
    chainLength=5
    relaxationSpeedScale=3
    rotationLimits [
     {limitMax=<180 100 180>, limitMin=<-180 -100 -180>}
     {limitMax=<180 100 180>, limitMin=<-180 -100 -180>}
     {limitMax=<180 100 180>, limitMin=<-180 -100 -180>}
     {limitMax=<180 100 180>, limitMin=<-180 -100 -180>}
     {limitMax=<180 100 180>, limitMin=<-180 -100 -180>}
    ]
    rotationOffsets [<0 0 0>, <0 0 0>, <0 0 0>, <0 0 0>, <0 0 0>]
    trailBone {boneName=dyn_hipCloth_07_r}
   }
   n12 = Trail controller - Bone: dyn_loinClothRear_06(ComponentPose=@n11, Alpha=1) {
    bLimitRotation=true
    chainLength=5
    relaxationSpeedScale=3
    rotationLimits [{limitMax=<180 20 180>}, {limitMax=<180 20 180>}, {limitMax=<180 20 180>}, {limitMax=<180 20 180>}, {limitMax=<180 180 180>}]
    rotationOffsets [<0 0 0>, <0 0 0>, <0 0 0>, <0 0 0>, <0 0 0>]
    trailBone {boneName=dyn_loinClothRear_06}
    trailRelaxationSpeed {editorCurveData {keys [
       <at 0 10 RCIM_Cubic>
       {arriveTangent=-5, interpMode=RCIM_Cubic, leaveTangent=-5, time=0.5876, value=8.6745}
       {arriveTangent=-12.6923, interpMode=RCIM_Cubic, leaveTangent=-12.6923, tangentMode=RCTM_User, time=1, value=5}
      ]}}
   }
   n13 = Trail controller - Bone: dyn_loinClothFront_07(ComponentPose=@n12, Alpha=1) {
    bLimitRotation=true
    chainLength=5
    relaxationSpeedScale=3
    rotationLimits [{limitMax=<180 0 180>}, {limitMax=<180 0 180>}, {limitMax=<180 0 180>}, {limitMax=<180 0 180>}, {limitMax=<180 0 180>}]
    rotationOffsets [<0 0 0>, <0 0 0>, <0 0 0>, <0 0 0>, <0 0 0>]
    trailBone {boneName=dyn_loinClothFront_07}
    trailRelaxationSpeed {editorCurveData {keys [<at 0 10 RCIM_Linear>,<at 0.4814 6.1572 RCIM_Linear>,<at 1 5 RCIM_Linear>]}}
   }
   n14 = "Transform (Modify) Bone - Bone: base_hipCloth_01_r"(ComponentPose=@n13, bAlphaBoolEnabled=True, Translation="0,0,0", Rotation="-3,0,0") {
    alphaInputType=Bool
    rotationMode=BMM_Additive
    rotationSpace=BCS_BoneSpace
    translationMode=BMM_Additive
    translationSpace=BCS_BoneSpace
    alphaBoolBlend {blendInTime=0.1, blendOutTime=0.1}
    boneToModify {boneName=base_hipCloth_01_r}
    rotation {pitch=-3}
   }
   n15 = "Transform (Modify) Bone - Bone: dyn_hipCloth_02_l"!(ComponentPose=@n14, bAlphaBoolEnabled:bool=True, Translation="0,4,-4", Rotation="3,0,10") -> (Pose) {
    alphaInputType=Bool
    rotationMode=BMM_Additive
    rotationSpace=BCS_BoneSpace
    translationMode=BMM_Additive
    translationSpace=BCS_BoneSpace
    alphaBoolBlend {blendInTime=0.1, blendOutTime=0.1}
    boneToModify {boneName=dyn_hipCloth_02_l}
    rotation {pitch=3, roll=10}
    translation {y=4, z=-4}
   }
   n16 = "Transform (Modify) Bone - Bone: dyn_loinClothFront_01"!(ComponentPose=@n15, bAlphaBoolEnabled:bool=True, Translation="0,0,0", Rotation="0,8,0") -> (Pose) {
    alphaInputType=Bool
    rotationMode=BMM_Additive
    rotationSpace=BCS_BoneSpace
    translationMode=BMM_Additive
    translationSpace=BCS_BoneSpace
    alphaBoolBlend {blendInTime=0.1, blendOutTime=0.1}
    boneToModify {boneName=dyn_loinClothFront_01}
    rotation {yaw=8}
   }
   n17 = "Transform (Modify) Bone - Bone: dyn_loinClothRear_01"!(ComponentPose=@n16, bAlphaBoolEnabled:bool=True, Translation="0,0,0", Rotation="0,-4,0") -> (Pose) {
    alphaInputType=Bool
    rotationMode=BMM_Additive
    rotationSpace=BCS_BoneSpace
    translationMode=BMM_Additive
    translationSpace=BCS_BoneSpace
    alphaBoolBlend {blendInTime=0.1, blendOutTime=0.1}
    boneToModify {boneName=dyn_loinClothRear_01}
    rotation {yaw=-4}
   }
   Save cached pose 'DynamicsCache'(Pose=Blend Poses by bool(BlendPose_0=Blend Poses by bool(BlendPose_0=@n2, BlendPose_1=@n7) {blendType=Linear}, BlendPose_1=Component To Local(ComponentPose=@n17)) {blendType=Linear}) {cacheName=DynamicsCache}
   n18 = Pose Driver - Source: upperarm_l!(SourcePose:Pose Link Structure=Copy Pose From Mesh() {bUseAttachedParent=true}) -> (Pose:Pose Link Structure) {
    is=title AnimGraphNode_PoseDriver
    poseAsset=&/M_MED_IceKing_Pistol_RBF_PoseAsset
    onlyDriveBones [
     {boneName=dyn_shoulderArmor_master_l}
     {boneName=dyn_shoulderArmor_offset_l}
     {boneName=dyn_shoulderArmor_l}
     {boneName=dyn_shoulderArmor_front_l}
     {boneName=dyn_shoulderArmor_back_l}
    ]
    poseTargets [
     {
      bApplyCustomCurve=false
      bIsHidden=false
      distanceMethod=DefaultMethod
      drivenName=Base
      functionType=DefaultFunction
      targetScale=0.5771
      boneTransforms [
       {targetRotation=<rot -64.7854 -3.089 6.0959>, targetTranslation=<15.5111 0 0>}
       {targetRotation=<rot -85.1915 -138.5625 143.0619>, targetTranslation=<4.3575 -0.1466 -2.3223>}
      ]
      customCurve {defaultValue=3.4028234663852886e+38, postInfinityExtrap=RCCE_Constant, preInfinityExtrap=RCCE_Constant, keys []}
      targetRotation=<rot 0 0 0>
     }
     {
      bApplyCustomCurve=false
      bIsHidden=false
      distanceMethod=DefaultMethod
      drivenName=Aim
      functionType=DefaultFunction
      targetScale=0.8339
      boneTransforms [
       {targetRotation=<rot -33.3286 69.4847 -74.886>, targetTranslation=<15.5111 0 0>}
       {targetRotation=<rot -82.5751 37.8255 -34.5856>, targetTranslation=<4.3575 -0.1466 -2.3223>}
      ]
      customCurve {defaultValue=3.4028234663852886e+38, postInfinityExtrap=RCCE_Constant, preInfinityExtrap=RCCE_Constant, keys []}
      targetRotation=<rot 0 0 0>
     }
     {
      bApplyCustomCurve=false
      bIsHidden=false
      distanceMethod=DefaultMethod
      drivenName=AimUp
      functionType=DefaultFunction
      targetScale=1
      boneTransforms [
       {targetRotation=<rot 9.7946 56.9767 -100.6551>, targetTranslation=<15.5111 0 0>}
       {targetRotation=<rot -49.8354 8.878 -0.0689>, targetTranslation=<4.3575 -0.1466 -2.3223>}
      ]
      customCurve {defaultValue=3.4028234663852886e+38, postInfinityExtrap=RCCE_Constant, preInfinityExtrap=RCCE_Constant, keys []}
      targetRotation=<rot 0 0 0>
     }
     {
      bApplyCustomCurve=false
      bIsHidden=false
      distanceMethod=DefaultMethod
      drivenName=AimDown
      functionType=DefaultFunction
      targetScale=0.5771
      boneTransforms [
       {targetRotation=<rot -77.3732 103.1484 -80.5252>, targetTranslation=<15.5111 0 0>}
       {targetRotation=<rot -85.6088 82.1547 -79.8546>, targetTranslation=<4.3575 -0.1466 -2.3223>}
      ]
      customCurve {defaultValue=3.4028234663852886e+38, postInfinityExtrap=RCCE_Constant, preInfinityExtrap=RCCE_Constant, keys []}
      targetRotation=<rot 0 0 0>
     }
    ]
    rBFParams {distanceMethod=Quaternion, radius=31.4759}
    sourceBones [{boneName=upperarm_l}, {boneName=clavicle_l}]
   }
   n19 = Pose Driver - Source: upperarm_r!(SourcePose:Pose Link Structure=@n18) -> (Pose:Pose Link Structure) {
    is=title AnimGraphNode_PoseDriver
    poseAsset=&M_MED_IceKing_Pistol_RBF_PoseAsset
    onlyDriveBones [
     {boneName=dyn_shoulderArmor_master_r}
     {boneName=dyn_shoulderArmor_offset_r}
     {boneName=dyn_shoulderArmor_r}
     {boneName=dyn_shoulderArmor_front_r}
     {boneName=dyn_shoulderArmor_back_r}
    ]
    poseTargets [
     {
      bApplyCustomCurve=false
      bIsHidden=false
      distanceMethod=DefaultMethod
      drivenName=Base
      functionType=DefaultFunction
      targetScale=1
      boneTransforms [
       {targetRotation=<rot -64.7854 -3.089 6.0959>, targetTranslation=<-15.5111 0 0>}
       {targetRotation=<rot -85.1915 41.4375 143.0619>, targetTranslation=<4.3575 -0.1466 2.3223>}
      ]
      customCurve {defaultValue=3.4028234663852886e+38, postInfinityExtrap=RCCE_Constant, preInfinityExtrap=RCCE_Constant, keys []}
      targetRotation=<rot 0 0 0>
     }
     {
      bApplyCustomCurve=false
      bIsHidden=false
      distanceMethod=DefaultMethod
      drivenName=Aim
      functionType=DefaultFunction
      targetScale=0.6023
      boneTransforms [
       {targetRotation=<rot -5.0115 105.0998 -58.2859>, targetTranslation=<-15.5111 0 0>}
       {targetRotation=<rot -68.9946 -127.6267 -53.9158>, targetTranslation=<4.3575 -0.1466 2.3223>}
      ]
      customCurve {defaultValue=3.4028234663852886e+38, postInfinityExtrap=RCCE_Constant, preInfinityExtrap=RCCE_Constant, keys []}
      targetRotation=<rot 0 0 0>
     }
     {
      bApplyCustomCurve=false
      bIsHidden=false
      distanceMethod=DefaultMethod
      drivenName=AimUp
      functionType=DefaultFunction
      targetScale=0.6023
      boneTransforms [
       {targetRotation=<rot 34.3086 96.9341 -83.9738>, targetTranslation=<-15.5111 0 0>}
       {targetRotation=<rot -68.9935 -127.6288 -53.9114>, targetTranslation=<4.3575 -0.1466 2.3223>}
      ]
      customCurve {defaultValue=3.4028234663852886e+38, postInfinityExtrap=RCCE_Constant, preInfinityExtrap=RCCE_Constant, keys []}
      targetRotation=<rot 0 0 0>
     }
     {
      bApplyCustomCurve=false
      bIsHidden=false
      distanceMethod=DefaultMethod
      drivenName=AimDown
      functionType=DefaultFunction
      targetScale=0.6095
      boneTransforms [
       {targetRotation=<rot -49.5228 109.2702 -46.6774>, targetTranslation=<-15.5111 0 0>}
       {targetRotation=<rot -71.2136 -118.598 -62.1124>, targetTranslation=<4.3575 -0.1466 2.3223>}
      ]
      customCurve {defaultValue=3.4028234663852886e+38, postInfinityExtrap=RCCE_Constant, preInfinityExtrap=RCCE_Constant, keys []}
      targetRotation=<rot 0 0 0>
     }
    ]
    rBFParams {distanceMethod=Quaternion, radius=38.3535}
    sourceBones [{boneName=upperarm_r}, {boneName=clavicle_r}]
   }
   n20 = Control Rig!(
    Source=@n19
    Alpha
    rBANEnabled:bool=false
    startCrouching:bool=false
    isKneeSliding:bool=false
    isBaseballSliding:bool=false
    isCrouching:bool=false
    isMovingForward:double=0
   ) -> (Pose) {controlRigAssetReference {blueprintRigClass=&/M_MED_IceKing_Body_CtrlRig.M_MED_IceKing_Body_CtrlRig_C}}
   Save cached pose 'CopyPoseCache'(Pose=@n20) {cacheName=CopyPoseCache}
   n21 = Get bCanPlayCustomAnimations() -> (bCanPlayCustomAnimations:bool)
   n22 = "Transform (Modify) Bone - Bone: dyn_hipCloth_01_r"!(
    ComponentPose=Local To Component(LocalPose=Use cached pose 'CopyPoseCache'())
    bAlphaBoolEnabled:bool=@n21
    Translation="0,0,0"
    Rotation="3,-17,0"
    Scale="1,1,1"
   ) -> (Pose) {
    alphaInputType=Bool
    rotationMode=BMM_Additive
    rotationSpace=BCS_BoneSpace
    translationMode=BMM_Additive
    translationSpace=BCS_BoneSpace
    alphaBoolBlend {blendOption=Linear}
    boneToModify {boneName=dyn_hipCloth_01_r}
    rotation {pitch=3, yaw=-17}
   }
   n23 = "Transform (Modify) Bone - Bone: dyn_hipCloth_01_l"!(ComponentPose=@n22, bAlphaBoolEnabled:bool=@n21, Translation="0,0,0", Rotation="2,-17,0", Scale="1,1,1") -> (Pose) {
    alphaInputType=Bool
    rotationMode=BMM_Additive
    rotationSpace=BCS_BoneSpace
    alphaBoolBlend {blendOption=Linear}
    boneToModify {boneName=dyn_hipCloth_01_l}
    rotation {pitch=2, yaw=-17}
   }
   n24 = "Transform (Modify) Bone - Bone: dyn_loinClothRear_01"!(ComponentPose=@n23, Alpha:float=1, Translation="0,0,0", Rotation="10,0.000,0.000", Scale:Vector="1,1,1") -> (Pose) {
    rotationMode=BMM_Additive
    rotationSpace=BCS_BoneSpace
    alphaBoolBlend {blendOption=Linear}
    boneToModify {boneName=dyn_loinClothRear_01}
    rotation {pitch=10}
   }
   n25 = "Transform (Modify) Bone - Bone: dyn_loinClothRear_01"!(ComponentPose=@n24, bAlphaBoolEnabled:bool=True, Translation="0,0,0", Rotation="15,0.000,0.000", Scale="1,1,1") -> (Pose) {
    alphaInputType=Bool
    bAlphaBoolEnabled=false
    rotationMode=BMM_Additive
    rotationSpace=BCS_BoneSpace
    scaleMode=BMM_Additive
    translationMode=BMM_Additive
    translationSpace=BCS_BoneSpace
    alphaBoolBlend {blendInTime=0.1, blendOption=Linear, blendOutTime=0.5}
    boneToModify {boneName=dyn_loinClothRear_01}
    rotation {pitch=15}
   }
   Save cached pose 'prerigid'(Pose=Component To Local(ComponentPose=@n25)) {cacheName=prerigid}
  ]}
 EventGraph {nodes [
   Event BlueprintUpdateAnimation()
   > n1 = Sequence!(execute) -> (then_0, then_1, then_2:Exec, then_3:Exec)
   n2 = GetAnimInstance(self=GetOwnerSkeletalMeshComponent())
   Is Valid(exec=@n1.then_1, InputObject=@n2)
   > Set vehicleemote(vehicleemote:bool=Montage_IsPlaying(self=@n2, Montage=/BRCosmetics/Animation/Game/MainPlayer/Emotes/Jupiter/Emote_Jupiter_CMM_M)) -> (Output_Get:bool)
   n3 = Branch(execute=@n1.then_2, Condition=Get bIsInFrontEnd())
   n4 = Get bIsPlayingEmote()
   n5 = Get bIsSkyDiving()
   > Set frontEndIdle(frontEndIdle:bool=NOT Boolean(A=OR Boolean(A=@n4, B=@n5))) -> (Output_Get:bool)
   Set frontEndIdle(execute=@n3.else, frontEndIdle=false)
   n6 = PlayerMovement(execute=@n1.then_3, AnimInstance=Self-Reference(), XDotProductCoefficient=1, YDotProductCoefficient=1, ZDotProductCoefficient=1)
   n7 = BreakVector(InVec=MakeVector(X=@n6, Y=@n6.YMovement))
   > Set strafeLeft float(strafeLeft float:double="Absolute (Float)"(A="Clamp (Float)"(Value=@n7, Min=-1, Max=0))) -> (Output_Get:double)
   > Set strafeRight float(strafeRight float:double="Clamp (Float)"(Value=@n7, Min=0, Max=1)) -> (Output_Get:double)
   > Set moveForward(moveForward:double="Clamp (Float)"(Value=@n7.Y, Min=0, Max=1)) -> (Output_Get:double)
   > Set moveBackward(moveBackward:double="Absolute (Float)"(A="Clamp (Float)"(Value=@n7.Y, Min=-1, Max=0))) -> (Output_Get:double)
   n8 = "Gravity Override State Function (Parameter Asset) (Nativized)"(
    execute=@n1.then_0
    CharacterPartAnimInstance=Self-Reference()
    GameTimeInSeconds=GetGameTimeInSeconds()
    Parameters=&/F_MED_IceKing_Parameters
   )
   > n9 = Set Gravity Override(Gravity Override=@n8)
   > Branch(Condition="Equal (Vector)"!(A:Vector=@n9, B:Vector, ErrorTolerance:float) -> (ReturnValue:bool=~) {is=title K2Node_PromotableOperator})
   > Set Gravity Override(Gravity Override="0,0,-400")
   Event OnParachuteOpened()
   > n10 = Delay(Duration=1.5)
   ResetDynamics(execute=[@n10.then, @n14.then, @n15.then, @n17.then, @n13.then], InTeleportType=TeleportPhysics)
   > ResetDynamics(InTeleportType=ResetPhysics)
   Event OnBeginCrouching()
   > Set startCrouch(startCrouch:bool=true) -> (Output_Get:bool)
   > Delay(Duration=0.2)
   > Set startCrouch(startCrouch=false)
   n11 = Event OnBeginTargeting() {is=title K2Node_Event}
   n12 = Get bIsCrouching()
   n13 = Branch(execute=[@n11.then, @n16.then], Condition=AND Boolean(A=Get bIsTargeting(), B=NOT Boolean(A=@n12)))
   n14 = Event OnBeginPlayingEmote() {is=title K2Node_Event}
   n15 = Event OnEndSkydiving()
   Event OnEndCrouching()
   > n16 = Set startCrouch(startCrouch=false)
   n17 = Event OnBeginSkydiving()
  ]}
]
variables [
 animdynamics_alpha {type=number, default=1}
 frontEndIdle {type=boolean}
 Gravity Override {type=Vector}
 moveBackward {type=number}
 moveForward {type=number}
 startCrouch {type=boolean}
 strafeLeft float {type=number}
 strafeRight float {type=number}
 vehicleemote {type=boolean}
]
