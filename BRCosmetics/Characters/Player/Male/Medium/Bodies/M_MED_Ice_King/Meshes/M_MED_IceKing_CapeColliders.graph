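# Physics asset for the Ice King medium male body, named as the cape's collision volumes.
# Capsules cover clavicles, neck, spine and pelvis, with a sphere on the head; arm, leg and foot bodies have no shapes.
# Dynamic hip-cloth and rear loincloth bones also carry capsules, including long ones turned crosswise, forming the lower-body collision in place of legs.
# Free-motion constraints link those cloth bodies in chains, with the rear chain attached to the pelvis.
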PhysicsAsset /BRCosmetics/Characters/Player/Male/Medium/Bodies/M_MED_Ice_King/Meshes/M_MED_IceKing_CapeColliders
bodies {
 calf_l []
 calf_r []
 clavicle_l [{shapeType=Capsule, length=7.7636, radius=11.1427, center=<9.9642 -0.7741 -1.5923>, rotation=<rot 75.9352 -0.0001 175.5622>}]
 clavicle_r [{shapeType=Capsule, length=7.7636, radius=11.1427, center=<-9.9642 0.7741 1.5923>, rotation=<rot 75.9351 -0.0001 -4.4378>}]
 dyn_hipCloth_03_l [{shapeType=Capsule, length=29, radius=8.25, shapeName=dyn_hipCloth_03_l_capsule, center=<0 0 -3.7013>, rotation=<rot -90 179.9902 -179.9899>}]
 dyn_hipCloth_03_r [{
   shapeType=Capsule
   length=29.5
   radius=7
   shapeName=dyn_hipCloth_03_r_capsule
   center=<-0.785 0 3.0995>
   rotation=<rot -90 179.9934 -179.9932>
  }]
 dyn_hipCloth_05_l [{
   shapeType=Capsule
   length=7.4196
   radius=9.6334
   shapeName=dyn_hipCloth_05_l_capsule
   center=<4.7706 -1.6641 -0.6387>
   rotation=<rot 0 -161.3313 -86.9203>
  }]
 dyn_hipCloth_05_r [{
   shapeType=Capsule
   length=7.4196
   radius=9.6334
   shapeName=dyn_hipCloth_05_r_capsule
   center=<-4.7706 1.6641 0.6387>
   rotation=<rot 0 -161.3312 -86.9203>
  }]
 dyn_hipCloth_06_l [{
   shapeType=Capsule
   length=32.75
   radius=10
   shapeName=dyn_hipCloth_06_l_capsule
   center=<-1.0395 0 -4.0396>
   rotation=<rot 90 -0.0197 -0.0199>
  }]
 dyn_hipCloth_06_r [{
   shapeType=Capsule
   length=34.75
   radius=10
   shapeName=dyn_hipCloth_06_r_capsule
   center=<1.3785 0.0017 2.8653>
   rotation=<rot 90 -179.367 -179.3671>
  }]
 dyn_hipCloth_07_l [{
   shapeType=Capsule
   length=8.0589
   radius=10.5957
   shapeName=dyn_hipCloth_07_l_capsule
   center=<0.8178 0.2408 -0.096>
   rotation=<rot 1.773 -167.2519 -66.3209>
  }]
 dyn_hipCloth_07_r [{
   shapeType=Capsule
   length=8.0589
   radius=10.5957
   shapeName=dyn_hipCloth_07_r_capsule
   center=<-0.8178 -0.2408 0.096>
   rotation=<rot 1.773 -167.2519 113.6791>
  }]
 dyn_loinClothRear_01 [{
   shapeType=Capsule
   length=2.7436
   radius=5.531
   shapeName=dyn_loinClothRear_01_capsule
   center=<1.9494 -0.0815 -1.2172>
   rotation=<rot 0 179.9116 -89.7961>
  }]
 dyn_loinClothRear_02 [{
   shapeType=Capsule
   length=9.1603
   radius=8.8563
   shapeName=dyn_loinClothRear_02_capsule
   center=<6.8801 -0.1447 -1.2709>
   rotation=<rot 0 179.328 -89.6888>
  }]
 dyn_loinClothRear_03 [{
   shapeType=Capsule
   length=9
   radius=9.3495
   shapeName=dyn_loinClothRear_03_capsule
   center=<4.3021 0.0663 -1.2389>
   rotation=<rot 0 0.2161 -88.7947>
  }]
 dyn_loinClothRear_04 [{
   shapeType=Capsule
   length=8
   radius=10.4692
   shapeName=dyn_loinClothRear_04_capsule
   center=<4.4286 0.0282 -1.2891>
   rotation=<rot 0 0.0952 -89.9162>
  }]
 dyn_loinClothRear_05 [{shapeType=Capsule, length=56.5, radius=10.25, shapeName=dyn_loinClothRear_05_capsule, center=<-1.4026 0 -3.4634>, rotation=<rot 90 0 0>}]
 dyn_loinClothRear_06 [{
   shapeType=Capsule
   length=13.8625
   radius=10.0833
   shapeName=dyn_loinClothRear_06_capsule
   center=<10.499 -0.1264 -1.3238>
   rotation=<rot 0 0 -90>
  }]
 foot_l []
 foot_r []
 head [{shapeType=Sphere, radius=17.3125, center=<11.028 -0.174 0>}]
 lowerarm_l []
 lowerarm_r []
 neck_02 [{shapeType=Capsule, length=13.4375, radius=7.3175, center=<-0.2642 -0.3408 0>, rotation=<rot 90 -171.1193 -175.3978>}]
 pelvis [{shapeType=Capsule, length=8.6562, radius=13.5388, center=<-4.7718 -1.3695 0>}]
 spine_01 [{shapeType=Capsule, length=12.7187, radius=12.9138, center=<-2.6215 -1.0385 0>}]
 spine_02 [
  {shapeType=Capsule, length=11.5937, radius=12.445, center=<-2.6215 0.1641 0.0001>}
  {shapeType=Capsule, length=29.875, radius=11.1013, center=<2.5024 0.4992 3.9825>, rotation=<rot 90 0.0049 0.0049>}
  {shapeType=Capsule, length=25.875, radius=15.1013, center=<2.4704 0.4991 -3.9825>, rotation=<rot 90 -0.0396 -0.0396>}
 ]
 spine_03 [{shapeType=Capsule, length=11.6562, radius=13.2575, center=<-1.8157 -1.112 0.0001>}]
 spine_04 [{shapeType=Capsule, length=15.6562, radius=16.6325, center=<2.2207 -2.1629 0.0002>}]
 spine_05 [
  {shapeType=Capsule, length=16.3437, radius=18.695, center=<-5.8354 0.378 0>}
  {shapeType=Capsule, length=9.2812, radius=20.5625, center=<0.7875 -0.5236 0>}
 ]
 thigh_l []
 thigh_r []
 upperarm_l []
 upperarm_r []
}
constraints [
 {
  bone1Name=dyn_hipCloth_07_l
  bone2Name=dyn_hipCloth_06_l
  swing1LimitDegrees=45
  swing1Motion=Free
  swing2LimitDegrees=45
  swing2Motion=Free
  twistLimitDegrees=45
  twistMotion=Free
 }
 {
  bone1Name=dyn_hipCloth_07_r
  bone2Name=dyn_hipCloth_06_r
  swing1LimitDegrees=45
  swing1Motion=Free
  swing2LimitDegrees=45
  swing2Motion=Free
  twistLimitDegrees=45
  twistMotion=Free
 }
 {
  bone1Name=dyn_loinClothRear_06
  bone2Name=dyn_loinClothRear_05
  swing1LimitDegrees=45
  swing1Motion=Free
  swing2LimitDegrees=45
  swing2Motion=Free
  twistLimitDegrees=45
  twistMotion=Free
 }
 {
  bone1Name=dyn_loinClothRear_05
  bone2Name=dyn_loinClothRear_04
  swing1LimitDegrees=45
  swing1Motion=Free
  swing2LimitDegrees=45
  swing2Motion=Free
  twistLimitDegrees=45
  twistMotion=Free
 }
 {
  bone1Name=dyn_loinClothRear_01
  bone2Name=pelvis
  swing1LimitDegrees=45
  swing1Motion=Free
  swing2LimitDegrees=45
  swing2Motion=Free
  twistLimitDegrees=45
  twistMotion=Free
 }
 {
  bone1Name=dyn_loinClothRear_02
  bone2Name=dyn_loinClothRear_01
  swing1LimitDegrees=45
  swing1Motion=Free
  swing2LimitDegrees=45
  swing2Motion=Free
  twistLimitDegrees=45
  twistMotion=Free
 }
 {
  bone1Name=dyn_loinClothRear_03
  bone2Name=dyn_loinClothRear_02
  swing1LimitDegrees=45
  swing1Motion=Free
  swing2LimitDegrees=45
  swing2Motion=Free
  twistLimitDegrees=45
  twistMotion=Free
 }
 {
  bone1Name=dyn_loinClothRear_04
  bone2Name=dyn_loinClothRear_03
  swing1LimitDegrees=45
  swing1Motion=Free
  swing2LimitDegrees=45
  swing2Motion=Free
  twistLimitDegrees=45
  twistMotion=Free
 }
 {
  bone1Name=dyn_hipCloth_06_r
  bone2Name=dyn_hipCloth_05_r
  swing1LimitDegrees=45
  swing1Motion=Free
  swing2LimitDegrees=45
  swing2Motion=Free
  twistLimitDegrees=45
  twistMotion=Free
 }
 {
  bone1Name=dyn_hipCloth_05_r
  bone2Name=dyn_hipCloth_03_r
  swing1LimitDegrees=45
  swing1Motion=Free
  swing2LimitDegrees=45
  swing2Motion=Free
  twistLimitDegrees=45
  twistMotion=Free
 }
 {
  bone1Name=dyn_hipCloth_06_l
  bone2Name=dyn_hipCloth_05_l
  swing1LimitDegrees=45
  swing1Motion=Free
  swing2LimitDegrees=45
  swing2Motion=Free
  twistLimitDegrees=45
  twistMotion=Free
 }
 {
  bone1Name=dyn_hipCloth_05_l
  bone2Name=dyn_hipCloth_03_l
  swing1LimitDegrees=45
  swing1Motion=Free
  swing2LimitDegrees=45
  swing2Motion=Free
  twistLimitDegrees=45
  twistMotion=Free
 }
]
