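# Post-copy animation Blueprint for the Ice King head cosmetic part, which follows the attached parent mesh and adjusts hood bones.
# Copy Pose From Mesh feeds the head's Control Rig.
# Two Bone Driven Controllers then drive each dyn_hood_clav bone's Y rotation from its matching clavicle, scaled and inverted.
# The EventGraph holds only an unconnected debug print of the right hood bone's rotation delta from reference pose.
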AnimBlueprint /BRCosmetics/Characters/Player/Male/Medium/Bodies/M_MED_Ice_King/Meshes/Parts/M_MED_IceKing_Head_AnimBP
target_skeleton=&/M_MED_IceKing_Head_Skeleton
graphs [
 AnimGraph {nodes [
   n1 = Control Rig(Source=Copy Pose From Mesh() {bUseAttachedParent=true}) {controlRigAssetReference {blueprintRigClass=&/M_MED_IceKing_Head_CtrlRig.M_MED_IceKing_Head_CtrlRig_C}}
   n2 = "dyn_hood_clav_l.rotateY = clavicle_l.rotateY * -2 - Bone Driven Controller"!(ComponentPose:Component Space Pose Link Structure=Local To Component(LocalPose=@n1), Alpha:float=1) -> (Pose:Component Space Pose Link Structure) {
    is=title AnimGraphNode_BoneDrivenController
    bAffectTargetRotationY=true
    multiplier=-2
    sourceComponent=RotationY
    alphaBoolBlend {blendOption=Linear}
    sourceBone {boneName=clavicle_l}
    targetBone {boneName=dyn_hood_clav_l}
   }
   n3 = "dyn_hood_clav_r.rotateY = clavicle_r.rotateY * -2 - Bone Driven Controller"!(ComponentPose:Component Space Pose Link Structure=@n2, Alpha:float=1) -> (Pose:Component Space Pose Link Structure) {
    is=title AnimGraphNode_BoneDrivenController
    bAffectTargetRotationY=true
    multiplier=-2
    sourceComponent=RotationY
    alphaBoolBlend {blendOption=Linear}
    sourceBone {boneName=clavicle_r}
    targetBone {boneName=dyn_hood_clav_r}
   }
   Output Pose(Result=Component To Local(ComponentPose=@n3)) {layerGroup=None}
  ]}
 EventGraph {nodes [
   n1 = GetDeltaTransformFromRefPose(self=GetPartSkeletalMeshComponent(), BoneName=dyn_hood_clav_r, BaseName=None)
   PrintString!(
    execute
    InString="To String (Rotator)"!(InRot:Rotator=BreakTransform(InTransform=@n1).Rotation) -> (ReturnValue:String)
    bPrintToScreen
    bPrintToLog
    TextColor
    Duration=0.03
    Key
   ) -> (then)
   Event BlueprintUpdateAnimation()
  ]}
]
variables []
